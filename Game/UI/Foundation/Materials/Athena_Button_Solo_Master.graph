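# UI material for an Athena solo button, drawing a procedural angled inner fill and border with translucency.
# Four rotated, offset UV edge ramps multiply into a quad mask, with OnHover blending each edge's angle and offset.
# The mask splits two colors, while Pressed and Selected-RGB-Fade push toward white.
# Selected and OnlyBorder choose visible regions for opacity, which a horizontal edge fade trims and AlphaOverride scales.
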Material /Game/UI/Foundation/Materials/Athena_Button_Solo_Master
graph {
 properties {
  bUsedWithStaticMesh=false
  blendMode=BLEND_Translucent
  materialDomain=MD_UI
  numCustomizedUVs=1
  pixelDepthOffsetMode=PDOM_Legacy
  usedShadingModels=Default Lit
 }
 nodes [
  n1 = VectorParameter() {group=Colors, parameterName=RGB_2, defaultValue=<rgba 0 1 0 1>}
  n2 = ScalarParameter() {group=States, parameterName=Pressed}
  n3 = ScalarParameter() {defaultValue=-0.0025, group=U-Edges, parameterName=U1-Angle}
  n4 = ScalarParameter() {defaultValue=-0.0025, group=U-Edges, parameterName=U1-Angle-Hovered, sortPriority=1}
  n5 = ScalarParameter() {group=States, parameterName=OnHover}
  n6 = TextureCoordinate() {coordinateIndex=4}
  n7 = ComponentMask(None=Rotator(Coordinate=@n6, Time=LinearInterpolate(A=@n3, B=@n4, Alpha=@n5) {constA=-0.0025}) {speed=6.2832}) {r=true}
  n8 = LinearInterpolate(
   A=ScalarParameter() {defaultValue=0.025, group=U-Edges, parameterName=U1-Offset}
   B=ScalarParameter() {defaultValue=0.025, group=U-Edges, parameterName=U1-Offset-Hovered, sortPriority=1}
   Alpha=@n5
  ) {constA=0.025}
  n9 = Saturate(None=Multiply(A=Subtract(A=@n7, B=@n8), B=ScalarParameter() {defaultValue=256, group=U-Edges, parameterName=U-Edge, sortPriority=1}))
  n10 = ScalarParameter() {defaultValue=-0.0113, group=V-Edges, parameterName=V1-Angle}
  n11 = ScalarParameter() {defaultValue=-0.0113, group=V-Edges, parameterName=V1-Angle-Hovered, sortPriority=1}
  n12 = ScalarParameter() {group=States, parameterName=OnHover}
  n13 = ComponentMask(None=Rotator(Coordinate=@n6, Time=LinearInterpolate(A=@n10, B=@n11, Alpha=@n12) {constA=-0.0113}) {speed=6.2832}) {g=true}
  n14 = LinearInterpolate(
   A=ScalarParameter() {defaultValue=0.125, group=V-Edges, parameterName=V1-Offset}
   B=ScalarParameter() {defaultValue=0.125, group=V-Edges, parameterName=V1-Offset-Hovered, sortPriority=1}
   Alpha=@n12
  ) {constA=0.125}
  n15 = Saturate(None=Multiply(A=Subtract(A=@n13, B=@n14), B=ScalarParameter() {defaultValue=64, group=V-Edges, parameterName=V-Edge, sortPriority=1}))
  n16 = ScalarParameter() {group=States, parameterName=OnHover}
  n17 = LinearInterpolate(
   A=ScalarParameter() {defaultValue=0.0025, group=U-Edges, parameterName=U2-Angle}
   B=ScalarParameter() {defaultValue=0.0025, group=U-Edges, parameterName=U2-Angle-Hovered, sortPriority=1}
   Alpha=@n16
  ) {constA=0.0025}
  n18 = OneMinus(None=ComponentMask(None=Rotator(Coordinate=@n6, Time=@n17) {speed=6.2832}) {r=true})
  n19 = LinearInterpolate(
   A=ScalarParameter() {defaultValue=0.025, group=U-Edges, parameterName=U2-Offset}
   B=ScalarParameter() {defaultValue=0.025, group=U-Edges, parameterName=U2-Offset-Hovered, sortPriority=1}
   Alpha=@n16
  ) {constA=0.025}
  n20 = Saturate(None=Multiply(A=Subtract(A=@n18, B=@n19), B=ScalarParameter() {defaultValue=256, group=U-Edges, parameterName=U-Edge}))
  n21 = ScalarParameter() {group=States, parameterName=OnHover}
  n22 = LinearInterpolate(
   A=ScalarParameter() {defaultValue=0.01, group=V-Edges, parameterName=V2-Angle}
   B=ScalarParameter() {defaultValue=0.01, group=V-Edges, parameterName=V2-Angle-Hovered, sortPriority=1}
   Alpha=@n21
  ) {constA=0.01}
  n23 = OneMinus(None=ComponentMask(None=Rotator(Coordinate=@n6, Time=@n22) {speed=6.2832}) {g=true})
  n24 = LinearInterpolate(
   A=ScalarParameter() {defaultValue=0.0905, group=V-Edges, parameterName=V2-Offset}
   B=ScalarParameter() {defaultValue=0.0905, group=V-Edges, parameterName=V2-Offset-Hovered, sortPriority=1}
   Alpha=@n21
  ) {constA=0.0905}
  n25 = Saturate(None=Multiply(A=Subtract(A=@n23, B=@n24), B=ScalarParameter() {defaultValue=64, group=V-Edges, parameterName=V-Edge}))
  n26 = Clamp(None=DotProduct(A=DotProduct(A=@n9, B=@n15), B=DotProduct(A=@n20, B=@n25)))
  LinearInterpolate(
   A=LinearInterpolate(
    A=VectorParameter() {group=Colors, parameterName=RGB_1, defaultValue=<rgba 1 0 0 1>}.RGB
    B=LinearInterpolate(A=@n1.RGB, Alpha=@n2)
    Alpha=@n26
   )
   B=Constant3Vector() {constant=<rgba 1 1 1 0>}
   Alpha=ScalarParameter() {group=States, parameterName=Selected-RGB-Fade}
  )
  n27 = OneMinus(None=@n26)
  n28 = LinearInterpolate(Alpha=@n26)
  n29 = ScalarParameter() {group=States, parameterName=Selected}
  n30 = LinearInterpolate(
   B=LinearInterpolate(
    A=VectorParameter() {group=Colors, parameterName=RGB_2, defaultValue=<rgba 0 1 0 1>}.A
    B=VectorParameter() {group=Colors, parameterName=RGB_1, defaultValue=<rgba 1 0 0 1>}.A
    Alpha=@n27
   )
   Alpha=LinearInterpolate(A=LinearInterpolate(B=@n28, Alpha=@n29) {constA=1}, B=@n27, Alpha=ScalarParameter() {group=States, parameterName=OnlyBorder})
  )
  n31 = Add(A=TextureCoordinate()) {constB=-0.5}
  n32 = MaterialFunctionCall(Float2=Multiply(A=Abs(None=@n31)) {constB=2}) {materialFunction=&BreakOutFloat2Components}
  n33 = Multiply(A=ScalarParameter() {parameterName=EdgeOffset}) {constB=0.001}
  n34 = Saturate(None=Multiply(A=Subtract(A=@n32.R, B=@n33) {constB=0}, B=ScalarParameter() {parameterName=EdgeStrength}))
  Multiply(A=Saturate(None=Subtract(A=@n30, B=@n34)), B=ScalarParameter() {defaultValue=1, group=States, parameterName=AlphaOverride, sortPriority=3})
 ]
}
