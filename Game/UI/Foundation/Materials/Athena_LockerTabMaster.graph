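# Translucent UI master material, named for locker tabs, that procedurally draws a two-color tab shape with soft, angled edges.
# Rotated UVs, folded around center for U and V (V on coordinate index 4), feed SmoothStep masks that combine to blend two RGBA parameters, separately for color and alpha.
# The UseOutsideAngles static switch, on by default, multiplies alpha by a rotated two-sided band mask to trim the outer silhouette.
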Material /Game/UI/Foundation/Materials/Athena_LockerTabMaster
graph {
 properties {
  bUsedWithStaticMesh=false
  blendMode=BLEND_Translucent
  materialDomain=MD_UI
  pixelDepthOffsetMode=PDOM_Legacy
  usedShadingModels=Default Lit
 }
 nodes [
  n1 = VectorParameter() {parameterName=RBGA1, defaultValue=<rgba 0 0.0578 0.5972 1>}
  n2 = VectorParameter() {parameterName=RBGA2, defaultValue=<rgba 1 1 1 0.5>}
  n3 = TextureCoordinate()
  n4 = Rotator(Coordinate=@n3, Time=ScalarParameter() {defaultValue=0.28, group=U, parameterName=Rotation_U})
  n5 = ComponentMask(None=Abs(None=Add(A=@n4) {constB=-0.5})) {r=true}
  n6 = ScalarParameter() {defaultValue=0.36, group=U, parameterName=Offset_U}
  n7 = Add(A=@n6, B=ScalarParameter() {defaultValue=0.01, group=U, parameterName=SSMargin_U, sliderMax=1})
  n8 = Rotator(Coordinate=TextureCoordinate() {coordinateIndex=4}, Time=ScalarParameter() {defaultValue=0.125, group=V, parameterName=Rotation_V})
  n9 = ComponentMask(None=Abs(None=Add(A=@n8) {constB=-0.5})) {g=true}
  n10 = ScalarParameter() {defaultValue=0.4625, group=V, parameterName=Offset_V}
  n11 = Add(A=@n10, B=ScalarParameter() {defaultValue=0.005, group=V, parameterName=SSMargin_V, sliderMax=1})
  n12 = Clamp(None=LinearInterpolate(
    A=MaterialFunctionCall(Alpha=@n5, Min=@n6, Max=@n7) {materialFunction=&SmoothStep}.Result
    Alpha=MaterialFunctionCall(Alpha=@n9, Min=@n10, Max=@n11) {materialFunction=&SmoothStep}.Result
   ))
  LinearInterpolate(A=@n1.RGB, B=@n2.RGB, Alpha=@n12)
  n13 = LinearInterpolate(A=@n1.A, B=@n2.A, Alpha=@n12)
  n14 = ScalarParameter() {defaultValue=0.375, group=OutsideShape, parameterName=Outside_Rotation}
  n15 = Rotator(Coordinate=@n3, Time=@n14)
  n16 = ScalarParameter() {defaultValue=0.1, group=OutsideShape, parameterName=Outside_OffsetLeft}
  n17 = ScalarParameter() {defaultValue=0.01, group=OutsideShape, parameterName=Outside_SSMargin, sliderMax=1}
  n18 = Add(A=ComponentMask(None=Rotator(Coordinate=@n3, Time=@n14)) {r=true}) {constB=-1}
  n19 = Multiply(A=Add(A=@n16, B=@n17)) {constB=-1}
  n20 = Add(A=@n19, B=@n17)
  n21 = LinearInterpolate(
   B=MaterialFunctionCall(Alpha=ComponentMask(None=@n15) {r=true}, Min=@n16, Max=Add(A=@n16, B=@n17)) {materialFunction=&SmoothStep}.Result
   Alpha=OneMinus(None=MaterialFunctionCall(Alpha=@n18, Min=@n19, Max=@n20) {materialFunction=&SmoothStep}.Result)
  )
  StaticSwitchParameter(True=LinearInterpolate(B=@n13, Alpha=Clamp(None=@n21)), False=@n13) {defaultValue=true, parameterName=UseOutsideAngles}
 ]
}
