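# UI material function that builds rotatable two-color linear and reflected gradients.
# It rotates the input UVs (texture coordinate 4 by default) with a Rotator driven by Gradient Rotation, then takes the X channel.
# SmoothStep between SS Min and SS Max shapes that value, or its absolute offset from 0.5 when reflected, into masks blending RGBA0 to RGBA1.
# It outputs both gradients, both masks and the two input colors.
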MaterialFunction /Game/UI/Foundation/Materials/Mf-UILinearGradientColor
graph {nodes [
  n1 = FunctionInput(Preview=Constant3Vector() {constant=<rgba 1 0 0 0>}) {bUsePreviewValueAsDefault=true, inputName=RGBA0}
  n2 = FunctionInput(Preview=Constant3Vector() {constant=<rgba 0 1 1 0>}) {bUsePreviewValueAsDefault=true, inputName=RGBA1, sortPriority=1}
  n3 = FunctionInput(Preview=TextureCoordinate() {coordinateIndex=4}) {bUsePreviewValueAsDefault=true, inputType=FunctionInput_Vector2}
  n4 = FunctionInput(Preview=Constant()) {bUsePreviewValueAsDefault=true, inputName=Gradient Rotation, inputType=FunctionInput_Scalar, sortPriority=4}
  n5 = FunctionInput(Preview=Constant()) {bUsePreviewValueAsDefault=true, inputName=SS Min, inputType=FunctionInput_Scalar, sortPriority=2}
  n6 = FunctionInput(Preview=Constant() {r=1}) {bUsePreviewValueAsDefault=true, inputName=SS Max, inputType=FunctionInput_Scalar, sortPriority=3}
  n7 = MaterialFunctionCall(Alpha=ComponentMask(None=Rotator(Coordinate=@n3, Time=@n4)) {r=true}, Min=@n5, Max=@n6) {materialFunction=&SmoothStep}
  FunctionOutput(None=LinearInterpolate(A=@n1, B=@n2, Alpha=@n7.Result)) {outputName=Linear Gradient}
  FunctionOutput(None=@n7.Result) {outputName=Linear Gradient Mask}
  n8 = ComponentMask(None=Rotator(Coordinate=@n3, Time=@n4)) {r=true}
  n9 = MaterialFunctionCall(Alpha=Abs(None=Add(A=@n8) {constB=-0.5}), Min=@n5, Max=@n6) {materialFunction=&SmoothStep}
  FunctionOutput(None=LinearInterpolate(A=@n1, B=@n2, Alpha=@n9.Result)) {outputName=Reflected Gradient, sortPriority=1}
  FunctionOutput(None=@n9.Result) {outputName=Reflected Gradient Mask, sortPriority=1}
  FunctionOutput(None=@n1) {outputName=RGBA0, sortPriority=2}
  FunctionOutput(None=@n2) {outputName=RGBA1, sortPriority=3}
 ]}
