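# Translucent UI master material for buttons, with static switches selecting between a texture and a procedural rounded shape.
# The texture path supports X-flip and gradient tint; the procedural path builds smoothstepped distance-field edges with gradient fill and optional drop shadow.
# An optional scalable icon is lerped over the color, then a shine swipe is color-dodged on top.
# Opacity comes from texture alpha or from the shape, with box or round cutouts and optional fill masks.
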Material /Game/UI/Foundation/Materials/M-Button-Master
graph {
 properties {
  bUsedWithStaticMesh=false
  blendMode=BLEND_Translucent
  materialDomain=MD_UI
  pixelDepthOffsetMode=PDOM_Legacy
  usedShadingModels=Default Lit
 }
 nodes [
  n1 = MaterialFunctionCall!(
   In
   RGBA0=VectorParameter() {parameterName=TintRGBA0, defaultValue=<rgba 1 0 0.9154 0>}.RGB
   RGBA1=VectorParameter() {parameterName=TintRGBA1, defaultValue=<rgba 0.0115 1 0.8703 0>}.RGB
   SS Min
   SS Max
   Gradient Rotation=ScalarParameter() {parameterName=Tint Rotation}
  ) {materialFunction=&/Mf-UILinearGradientColor}
  n2 = StaticSwitchParameter(True=@n1.RGBA0, False=@n1.RGBA0) {parameterName=Solid Tint}
  n3 = MaterialFunctionCall() {materialFunction=&GetUserInterfaceUV}
  n4 = TextureSampleParameter2D(UVs=StaticSwitchParameter(True=OneMinus(None=@n3.9-Slice UV), False=@n3.9-Slice UV) {parameterName=Texture_FlipX}) {parameterName=Texture, texture=/Game/UI/Foundation/Textures/Buttons/T-Button-Primary-Base-64}
  n5 = VectorParameter() {parameterName=Drop Shadow, defaultValue=<rgba 0.0976 0.162 0.3325 1>}
  n6 = VectorParameter() {parameterName=RGBA0, defaultValue=<rgba 0.7084 0.7913 0.8963 1>}
  n7 = VectorParameter() {parameterName=RGBA1, defaultValue=<rgba 0.4397 0.5841 0.7991 1>}
  n8 = ComponentMask(None=Rotator(
    Coordinate=MaterialFunctionCall() {materialFunction=&GetUserInterfaceUV}.Normalized UV
    Time=ScalarParameter() {group=Linear Gradient Settings, parameterName=Rotation_1}
   )) {g=true}
  n9 = ScalarParameter() {group=Linear Gradient Settings, parameterName=LG SS Min_1}
  n10 = ScalarParameter() {defaultValue=1, group=Linear Gradient Settings, parameterName=LG SS Max_1}
  n11 = Abs(None=Add(A=TextureCoordinate()) {constB=-0.5})
  n12 = ScalarParameter() {defaultValue=0.3, parameterName=Size}
  n13 = Subtract(A=@n11, B=@n12)
  n14 = Constant()
  n15 = ScalarParameter() {defaultValue=0.18, parameterName=SS Min}
  n16 = ScalarParameter() {defaultValue=0.15, parameterName=SS Max}
  n17 = MaterialFunctionCall(
   Alpha=MaterialFunctionCall(Vector 2=Max(A=@n13, B=@n14)) {materialFunction=/Engine/Functions/Engine_MaterialFunctions02/Utility/VectorLength}.V2 Length
   Min=@n15
   Max=@n16
  ) {materialFunction=&SmoothStep}
  n18 = LinearInterpolate(
   A=@n5.RGB
   B=LinearInterpolate(
    A=@n6.RGB
    B=@n7.RGB
    Alpha=MaterialFunctionCall(Alpha=@n8, Min=@n9, Max=@n10) {materialFunction=&SmoothStep}.Result
   )
   Alpha=@n17.Result
  )
  n19 = Floor(None=Add(A=TextureCoordinate()) {constB=0.7})
  n20 = Subtract(
   A=@n11
   B=AppendVector(A=@n12, B=ScalarParameter() {defaultValue=0.2193, parameterName=Drop Shadow Distance, sliderMax=0.35, sliderMin=-0.05})
  )
  n21 = Constant()
  n22 = MaterialFunctionCall(
   Alpha=MaterialFunctionCall(Vector 2=Max(A=@n20, B=@n21)) {materialFunction=&VectorLength}.V2 Length
   Min=@n15
   Max=@n16
  ) {materialFunction=&SmoothStep}
  n23 = Multiply(
   A=Multiply(A=ComponentMask(None=@n19) {g=true}, B=OneMinus(None=@n22.Result))
   B=ScalarParameter() {defaultValue=1, parameterName=DropShadow Alpha, sliderMax=1}
  )
  n24 = MaterialFunctionCall() {materialFunction=&GetUserInterfaceUV}
  n25 = ScalarParameter() {defaultValue=1, group=Linear Gradient Settings, parameterName=Rotation}
  n26 = MaterialFunctionCall(
   Alpha=ComponentMask(None=Rotator(Coordinate=@n24.Normalized UV, Time=@n25)) {r=true}
   Min=ScalarParameter() {group=Linear Gradient Settings, parameterName=LG SS Min}
   Max=ScalarParameter() {defaultValue=1, group=Linear Gradient Settings, parameterName=LG SS Max}
  ) {materialFunction=&SmoothStep}
  n27 = StaticSwitchParameter(
   True=StaticSwitchParameter(True=Multiply(A=@n2, B=@n4.RGB), False=@n4.RGB) {parameterName=Tint Texture}
   False=StaticSwitchParameter(True=LinearInterpolate(A=@n18, B=@n5.RGB, Alpha=@n23), False=LinearInterpolate(A=@n6.RGB, B=@n7.RGB, Alpha=@n26.Result)) {defaultValue=true, parameterName=Drop Shadow}
  ) {parameterName=Use Texture}
  n28 = VectorParameter() {parameterName=IconRGBA}
  n29 = MaterialFunctionCall() {materialFunction=&GetUserInterfaceUV}
  n30 = ComponentMask(None=VectorParameter() {group=Icon, parameterName=IconPosition}.RGB) {g=true, r=true}
  n31 = Divide(
   A=ComponentMask(None=VectorParameter() {group=Icon, parameterName=IconSize, defaultValue=<rgba 52 52 0 0>}.RGB) {g=true, r=true}
   B=@n29.Pixel Size
  )
  n32 = ScalarParameter() {defaultValue=1, parameterName=Icon Scale}
  n33 = TextureSampleParameter2D(UVs=MaterialFunctionCall(UVs=Add(A=@n29.Normalized UV, B=@n30), Texture Scale=Multiply(A=@n31, B=@n32)) {materialFunction=&ScaleUVsByCenter}.UVs) {group=Icon, parameterName=Center Icon, texture=/Game/UI/Foundation/Textures/Icons/Locks/T-Icon-Lock-64}
  MaterialFunctionCall!(
   Blend=Multiply(
    A=ScalarParameter() {parameterName=Shine, sliderMax=1}
    B=MaterialFunctionCall!(In_Uv) {materialFunction=&/M-Function-ShineSwipe}.Result
   )
   Base=StaticSwitchParameter(True=LinearInterpolate(A=@n27, B=@n28.RGB, Alpha=@n33.A), False=@n27) {group=Icon, parameterName=Use Icon}
  ) {materialFunction=/Engine/Functions/Engine_MaterialFunctions03/Blends/Blend_ColorDodge}
  n34 = TextureCoordinate()
  n35 = Subtract(A=Abs(None=Add(A=@n34) {constB=-0.5}), B=ScalarParameter() {defaultValue=0.2, parameterName=Box Size})
  MaterialFunctionCall(Alpha=@n35, Min=Constant() {r=0.1}, Max=Constant() {r=0.1}) {materialFunction=&SmoothStep}
  n36 = Abs(None=Add(A=TextureCoordinate()) {constB=-0.5})
  n37 = ScalarParameter() {defaultValue=0.2, parameterName=Box Size_1}
  MaterialFunctionCall(
   Alpha=MaterialFunctionCall(Vector 2=Subtract(A=@n36, B=@n37)) {materialFunction=&VectorLength}.V2 Length
   Min=Constant() {r=0.125}
   Max=Constant() {r=0.15}
  ) {materialFunction=&SmoothStep}
  n38 = MaterialFunctionCall(Vector 2=Ceil(None=@n35)) {materialFunction=&VectorLength}
  n39 = Clamp(None=Add(A=@n38.V2 Length, B=ScalarParameter() {defaultValue=0.2, parameterName=Box Cutout Alpha}))
  n40 = Clamp(None=@n38.V2 Length)
  n41 = MaterialFunctionCall(
   Alpha=Add(A=TextureCoordinate(), B=ScalarParameter() {defaultValue=-0.5, parameterName=Mask Fill Offset}) {constB=0.1}
   Min=ScalarParameter() {defaultValue=0.125, parameterName=Mask SS Min}
   Max=ScalarParameter() {defaultValue=0.14, parameterName=Mask SS Max}
  ) {materialFunction=&SmoothStep}
  n42 = OneMinus(None=@n40)
  n43 = StaticSwitchParameter(True=Multiply(A=@n39, B=LinearInterpolate(A=@n40, B=@n41.Result, Alpha=@n42)), False=@n39) {parameterName=Use Mask for Fill}
  n44 = ScalarParameter() {defaultValue=0.2, parameterName=Size_1}
  n45 = MaterialFunctionCall(Vector 2=Max(A=Subtract(A=@n11, B=@n44), B=Constant())) {materialFunction=&VectorLength}
  n46 = ScalarParameter() {defaultValue=0.14, parameterName=Inset SS Min}
  n47 = ScalarParameter() {defaultValue=0.11, parameterName=Inset SS Max}
  n48 = Add(
   A=OneMinus(None=MaterialFunctionCall(Alpha=@n45.V2 Length, Min=@n46, Max=@n47) {materialFunction=&SmoothStep}.Result)
   B=ScalarParameter() {parameterName=Round Cutout Alpha}
  )
  n49 = Add(A=TextureCoordinate(), B=ScalarParameter() {defaultValue=-0.5, parameterName=Mask Fill Offset}) {constB=0.1}
  n50 = ScalarParameter() {defaultValue=0.125, parameterName=Mask SS Min}
  n51 = ScalarParameter() {defaultValue=0.14, parameterName=Mask SS Max}
  n52 = StaticSwitchParameter(
   True=Multiply(
    A=@n48
    B=MaterialFunctionCall(Alpha=@n49, Min=@n50, Max=@n51) {materialFunction=&SmoothStep}.Result
   )
   False=@n48
  ) {parameterName=Use Mask Fill}
  n53 = StaticSwitchParameter(True=Multiply(A=@n17.Result, B=Clamp(None=@n52)), False=@n17.Result) {parameterName=Cutout Round}
  StaticSwitchParameter(
   True=Multiply(A=ScalarParameter() {defaultValue=1, parameterName=Alpha}, B=@n4.A)
   False=Multiply(
    A=StaticSwitchParameter(True=@n43, False=@n53) {parameterName=Box Cutout}
    B=ScalarParameter() {defaultValue=1, parameterName=Alpha, sliderMax=1}
   )
  ) {parameterName=Use Texture}
 ]
}
