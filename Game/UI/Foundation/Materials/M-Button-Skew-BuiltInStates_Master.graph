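# Translucent UI master material for a skewed button with built-in selection and shape states, drawn procedurally without textures.
# Four rotated-UV SmoothStep edges form the quad mask, with MasterShapeChange blending each edge between A and B angles.
# A center-scaled inner copy carries the flat or static-switched gradient fill; the remaining rim shows OuterHighlightColor, scaled by Selection.
# Alpha scales the combined opacity, and separate offset edges build a drop-shadow mask.
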Material /Game/UI/Foundation/Materials/M-Button-Skew-BuiltInStates_Master
graph {
 properties {
  bUsedWithStaticMesh=false
  blendMode=BLEND_Translucent
  materialDomain=MD_UI
  pixelDepthOffsetMode=PDOM_Legacy
  usedShadingModels=Default Lit
 }
 nodes [
  n1 = MaterialFunctionCall() {materialFunction=&GetUserInterfaceUV}
  n2 = ScalarParameter() {parameterName=MasterShapeChange}
  n3 = LinearInterpolate(A=ScalarParameter() {defaultValue=12.4, parameterName=Left_A}, B=ScalarParameter() {defaultValue=12.6887, parameterName=Left_B}, Alpha=@n2) {constAlpha=0}
  n4 = ComponentMask(None=Rotator(Coordinate=@n1.Normalized UV, Time=@n3)) {r=true}
  n5 = ScalarParameter() {defaultValue=0.49, parameterName=SS Min - HEdges}
  n6 = ScalarParameter() {defaultValue=0.51, parameterName=SS Max}
  n7 = MaterialFunctionCall(Alpha=Multiply(A=Add(A=@n4) {constB=0.225}) {constB=2}, Min=@n5, Max=@n6) {materialFunction=&SmoothStep}
  n8 = LinearInterpolate(A=ScalarParameter() {defaultValue=0.07, parameterName=Right_A}, B=ScalarParameter() {defaultValue=-0.3, parameterName=Right_B}, Alpha=@n2) {constAlpha=0}
  n9 = Add(A=ComponentMask(None=Rotator(Coordinate=@n1.Normalized UV, Time=@n8)) {r=true}) {constB=0.225}
  n10 = Multiply(A=MaterialFunctionCall(Alpha=Multiply(A=@n9) {constB=2}, Min=@n5, Max=@n6) {materialFunction=&SmoothStep}.Result)
  n11 = LinearInterpolate(A=ScalarParameter() {defaultValue=0.5, parameterName=Top_A}, B=ScalarParameter() {defaultValue=0.5, parameterName=Top_B}, Alpha=@n2) {constAlpha=0}
  n12 = ComponentMask(None=Rotator(Coordinate=@n1.Normalized UV, Time=@n11) {speed=6.2832}) {g=true}
  n13 = ScalarParameter() {defaultValue=0.49, parameterName=SS Min - VEdge}
  n14 = MaterialFunctionCall(Alpha=Multiply(A=Add(A=@n12) {constB=0.125}) {constB=2}, Min=@n13, Max=@n6) {materialFunction=&SmoothStep}
  n15 = LinearInterpolate(A=ScalarParameter() {parameterName=Bot_A}, B=ScalarParameter() {parameterName=Bot_B}, Alpha=@n2) {constAlpha=0}
  n16 = Add(A=ComponentMask(None=Rotator(Coordinate=@n1.Normalized UV, Time=@n15) {speed=6.2832}) {g=true}) {constB=0.2}
  n17 = Multiply(A=MaterialFunctionCall(Alpha=Multiply(A=@n16) {constB=2}, Min=@n13, Max=@n6) {materialFunction=&SmoothStep}.Result)
  n18 = LinearInterpolate(B=LinearInterpolate(B=Multiply(A=@n7.Result), Alpha=@n10), Alpha=LinearInterpolate(B=Multiply(A=@n14.Result), Alpha=@n17))
  n19 = Rotator(Coordinate=@n1.Normalized UV, Time=Constant() {r=12.25})
  n20 = Multiply(A=Add(A=ComponentMask(None=@n19) {g=true}) {constB=0.16}) {constB=2}
  n21 = ScalarParameter() {defaultValue=0.48, parameterName=SS Min - Shadow - VEdge}
  n22 = LinearInterpolate(
   B=Multiply(A=MaterialFunctionCall(Alpha=@n20, Min=@n21, Max=@n6) {materialFunction=&SmoothStep}.Result)
   Alpha=@n17
  )
  n23 = Rotator(Coordinate=@n1.Normalized UV, Time=Constant() {r=12.5})
  n24 = Multiply(A=Add(A=ComponentMask(None=@n23) {r=true}) {constB=0.235}) {constB=2}
  n25 = ScalarParameter() {defaultValue=0.49, parameterName=SS Min - Shadow- HEdges}
  n26 = LinearInterpolate(
   B=Multiply(A=MaterialFunctionCall(Alpha=@n24, Min=@n25, Max=@n6) {materialFunction=&SmoothStep}.Result)
   Alpha=@n10
  )
  Add(
   A=@n18
   B=Multiply(A=LinearInterpolate(B=@n22, Alpha=@n26), B=ScalarParameter() {defaultValue=0.5, parameterName=DropShadow Alpha}) {constB=0.5}
  )
  n27 = VectorParameter() {parameterName=OuterHighlightColor, defaultValue=<rgba 1 0.8958 0 1>}
  n28 = MaterialFunctionCall(
   UVs=MaterialFunctionCall() {materialFunction=&GetUserInterfaceUV}.Normalized UV
   Texture Scale=VectorParameter() {parameterName=InnerShapeScale, defaultValue=<rgba 0.94 0.9 0 1>}.RGB
  ) {materialFunction=&ScaleUVsByCenter}
  n29 = ComponentMask(None=Rotator(Coordinate=@n28.UVs, Time=@n3)) {r=true}
  n30 = ScalarParameter() {defaultValue=0.49, parameterName=SS Min - HEdges_1}
  n31 = ScalarParameter() {defaultValue=0.51, parameterName=SS Max_1}
  n32 = MaterialFunctionCall(Alpha=Multiply(A=Add(A=@n29) {constB=0.225}) {constB=2}, Min=@n30, Max=@n31) {materialFunction=&SmoothStep}
  n33 = ComponentMask(None=Rotator(Coordinate=@n28.UVs, Time=@n8)) {r=true}
  n34 = MaterialFunctionCall(Alpha=Multiply(A=Add(A=@n33) {constB=0.225}) {constB=2}, Min=@n30, Max=@n31) {materialFunction=&SmoothStep}
  n35 = ComponentMask(None=Rotator(Coordinate=@n28.UVs, Time=@n11) {speed=6.2832}) {g=true}
  n36 = ScalarParameter() {defaultValue=0.49, parameterName=SS Min - VEdge_1}
  n37 = MaterialFunctionCall(Alpha=Multiply(A=Add(A=@n35) {constB=0.125}) {constB=2}, Min=@n36, Max=@n31) {materialFunction=&SmoothStep}
  n38 = ComponentMask(None=Rotator(Coordinate=@n28.UVs, Time=@n15) {speed=6.2832}) {g=true}
  n39 = MaterialFunctionCall(Alpha=Multiply(A=Add(A=@n38) {constB=0.2}) {constB=2}, Min=@n36, Max=@n31) {materialFunction=&SmoothStep}
  n40 = LinearInterpolate(
   B=LinearInterpolate(B=Multiply(A=@n32.Result), Alpha=Multiply(A=@n34.Result))
   Alpha=LinearInterpolate(B=Multiply(A=@n37.Result), Alpha=Multiply(A=@n39.Result))
  )
  n41 = OneMinus(None=@n40)
  n42 = Multiply(A=Multiply(A=@n27.RGB, B=Multiply(A=@n41, B=@n18)), B=Multiply(A=@n27.A, B=ScalarParameter() {parameterName=Selection}))
  n43 = VectorParameter() {parameterName=FillColor, defaultValue=<rgba 0 0.6984 1 1>}
  n44 = MaterialFunctionCall() {materialFunction=&GetUserInterfaceUV}
  n45 = LinearInterpolate(
   A=@n43.RGB
   B=VectorParameter() {parameterName=FillColor_B, defaultValue=<rgba 0.0608 0.0682 0.7604 1>}.RGB
   Alpha=Power(Base=ComponentMask(None=@n44.Normalized UV) {r=true}, Exp=ScalarParameter() {defaultValue=1, parameterName=GradientExp})
  )
  Add(A=@n42, B=Multiply(A=@n40, B=StaticSwitchParameter(True=@n45, False=@n43.RGB) {parameterName=UseGradientFill}))
  Constant() {r=0.07}
  Constant() {r=12.4}
  Constant()
  Constant() {r=0.5}
  MaterialFunctionCall() {materialFunction=&DebugTimeSine}
  n46 = Multiply(A=@n43.A, B=@n42)
  Multiply(A=Clamp(None=Add(A=@n40, B=@n46)), B=ScalarParameter() {defaultValue=1, parameterName=Alpha, sliderMax=1})
  VectorParameter() {parameterName=Drop Shadow, defaultValue=<rgba 0.0203 0.0409 0.1301 1>}
 ]
}
