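# Translucent UI master material that draws a skewed button procedurally, with drop shadow and shine swipe.
# Rotated UVs through SmoothStep form four soft edges for the button mask and a shadow mask, combined for opacity.
# The button mask blends shadow color with a flat or gradient fill, then an optional rotated sine shine band is added.
# SwipeLoc positions the shine; the Time-based Cycles On/Off logic and Shine Speed do not feed it.
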Material /Game/UI/Foundation/Materials/M-Button-Skewed2-Master
graph {
 properties {
  bUsedWithStaticMesh=false
  blendMode=BLEND_Translucent
  materialDomain=MD_UI
  pixelDepthOffsetMode=PDOM_Legacy
  usedShadingModels=Default Lit
 }
 nodes [
  n1 = Constant() {r=1}
  n2 = Time()
  n3 = Divide(A=ScalarParameter() {defaultValue=3, group=Shine, parameterName=Cycles On})
  n4 = Constant() {r=2}
  n5 = Multiply(A=Divide(A=ScalarParameter() {defaultValue=6, group=Shine, parameterName=Cycles Off}), B=Constant() {r=2})
  n6 = Fmod(A=@n2, B=Add(A=Multiply(A=@n3, B=@n4), B=@n5))
  n7 = Constant()
  n8 = Constant() {r=-1}
  If(
   A=Add(A=Fmod(A=@n2, B=@n1), B=If(A=@n6, B=@n5, "A > B"=@n7, "A == B"=@n8, "A < B"=@n8)) {constB=-1}
   B=Constant()
   "A > B"=Constant() {r=1}
   "A < B"=Constant()
  )
  n9 = VectorParameter() {parameterName=Gradient Fill_Top, defaultValue=<rgba 1 1 0.43 1>}
  n10 = VectorParameter() {parameterName=Gradient Fill_Bottom, defaultValue=<rgba 1 0.9167 0 1>}
  n11 = ComponentMask(None=TextureCoordinate()) {g=true}
  n12 = TextureCoordinate()
  n13 = ComponentMask(None=Rotator(Coordinate=@n12, Time=ScalarParameter() {defaultValue=12.4, parameterName=Angle_Right, sliderMax=12.9, sliderMin=12.3})) {r=true}
  n14 = ScalarParameter() {defaultValue=0.48, parameterName=SS Min - HEdges}
  n15 = ScalarParameter() {defaultValue=0.52, parameterName=SS Max}
  n16 = MaterialFunctionCall(Alpha=Multiply(A=Add(A=@n13) {constB=0.225}) {constB=2}, Min=@n14, Max=@n15) {materialFunction=&SmoothStep}
  n17 = ScalarParameter() {defaultValue=0.24, parameterName=Angle_Left, sliderMax=0.3, sliderMin=-0.3}
  n18 = Add(A=ComponentMask(None=Rotator(Coordinate=@n12, Time=@n17)) {r=true}) {constB=0.225}
  n19 = Multiply(A=MaterialFunctionCall(Alpha=Multiply(A=@n18) {constB=2}, Min=@n14, Max=@n15) {materialFunction=&SmoothStep}.Result)
  n20 = ComponentMask(None=Rotator(Coordinate=@n12, Time=Constant() {r=0.5}) {speed=6.2832}) {g=true}
  n21 = ScalarParameter() {defaultValue=0.47, parameterName=SS Min - VEdge}
  n22 = MaterialFunctionCall(Alpha=Multiply(A=Add(A=@n20) {constB=0.125}) {constB=2}, Min=@n21, Max=@n15) {materialFunction=&SmoothStep}
  n23 = Constant()
  n24 = Add(A=ComponentMask(None=Rotator(Coordinate=@n12, Time=@n23) {speed=6.2832}) {g=true}) {constB=0.2}
  n25 = Multiply(A=MaterialFunctionCall(Alpha=Multiply(A=@n24) {constB=2}, Min=@n21, Max=@n15) {materialFunction=&SmoothStep}.Result)
  n26 = LinearInterpolate(B=LinearInterpolate(B=Multiply(A=@n16.Result), Alpha=@n19), Alpha=LinearInterpolate(B=Multiply(A=@n22.Result), Alpha=@n25))
  n27 = LinearInterpolate(
   A=VectorParameter() {parameterName=Drop Shadow, defaultValue=<rgba 0.0203 0.0409 0.1301 1>}.RGB
   B=LinearInterpolate(
    A=VectorParameter() {parameterName=Button Fill, defaultValue=<rgba 1 0.9167 0 1>}.RGB
    B=LinearInterpolate(A=@n9.RGB, B=@n10.RGB, Alpha=@n11)
    Alpha=ScalarParameter() {parameterName=bGradientFill, sliderMax=1}
   )
   Alpha=@n26
  )
  n28 = Constant()
  n29 = VectorParameter() {group=Shine, parameterName=ShineRGBA, defaultValue=<rgba 1 1 1 0.5>}
  n30 = Multiply(A=ScalarParameter() {parameterName=SwipeLoc, sliderMax=2})
  n31 = Add(
   A=Rotator(Time=ScalarParameter() {defaultValue=0.15, group=Shine, parameterName=Shine Rotation}) {centerX=0, centerY=0}
   B=AppendVector(A=OneMinus(None=@n30), B=Constant())
  )
  n32 = ScalarParameter() {defaultValue=0.75, group=Shine, parameterName=Shine SS Min}
  n33 = MaterialFunctionCall(
   Alpha=Sine(None=ComponentMask(None=@n31) {r=true}) {period=2}
   Min=@n32
   Max=Add(A=@n32, B=ScalarParameter() {defaultValue=0.25, group=Shine, parameterName=Shine SS Max})
  ) {materialFunction=&SmoothStep}
  LinearInterpolate(
   A=@n27
   B=Add(A=LinearInterpolate(A=@n28, B=@n29.RGB, Alpha=@n33.Result), B=@n27)
   Alpha=ScalarParameter() {defaultValue=1, parameterName=UseShineSwipe}
  )
  Multiply()
  n34 = ComponentMask(None=Rotator(Coordinate=@n12, Time=Constant() {r=12.25})) {g=true}
  n35 = MaterialFunctionCall(
   Alpha=Multiply(A=Add(A=@n34) {constB=0.16}) {constB=2}
   Min=ScalarParameter() {defaultValue=0.48, parameterName=SS Min - Shadow - VEdge}
   Max=@n15
  ) {materialFunction=&SmoothStep}
  n36 = ComponentMask(None=Rotator(Coordinate=@n12, Time=ScalarParameter() {defaultValue=12.3, parameterName=Shadow_Angle_R, sliderMax=12.9, sliderMin=12.3})) {r=true}
  n37 = MaterialFunctionCall(
   Alpha=Multiply(A=Add(A=@n36) {constB=0.235}) {constB=2}
   Min=ScalarParameter() {defaultValue=0.49, parameterName=SS Min - Shadow- HEdges}
   Max=@n15
  ) {materialFunction=&SmoothStep}
  n38 = LinearInterpolate(B=LinearInterpolate(B=Multiply(A=@n35.Result), Alpha=@n25), Alpha=LinearInterpolate(B=Multiply(A=@n37.Result), Alpha=@n19))
  n39 = ScalarParameter() {defaultValue=0.5, parameterName=DropShadow Alpha}
  Multiply(A=Add(A=@n26, B=Multiply(A=@n38, B=@n39) {constB=0.5}), B=ScalarParameter() {defaultValue=1, parameterName=Alpha, sliderMax=1})
  ScalarParameter() {defaultValue=4, group=Shine, parameterName=Shine Speed}
  Time()
 ]
}
