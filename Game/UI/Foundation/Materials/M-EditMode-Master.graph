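# Translucent UI master material for edit-mode slot highlights, with static switches choosing an item card border, a circular gadget ring or a banner mask.
# Card borders are solid or marching dashes, using opposed panners swapped per UV half; rings subtract two radial gradients, optionally cut by a rotating dash texture.
# That shape drives tint, optional dark overlay and opacity, within a square slot mask.
# Opacity can pulse and add a centred, size-animated icon, defaulting to a lock.
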Material /Game/UI/Foundation/Materials/M-EditMode-Master
graph {
 properties {
  bUsedWithStaticMesh=false
  blendMode=BLEND_Translucent
  materialDomain=MD_UI
  pixelDepthOffsetMode=PDOM_Legacy
  usedShadingModels=Default Lit
 }
 nodes [
  n1 = MaterialFunctionCall() {materialFunction=&GetUserInterfaceUV}
  n2 = Divide(A=ComponentMask(None=VectorParameter() {parameterName=IconSize, defaultValue=<rgba 52 52 0 0>}.RGB) {g=true, r=true}, B=@n1.Pixel Size)
  n3 = Time()
  n4 = ScalarParameter() {defaultValue=1, parameterName=Icon Size Anim Speed}
  n5 = LinearInterpolate(
   A=ScalarParameter() {defaultValue=1, parameterName=Icon Scale}
   B=MaterialFunctionCall!(
    Sine Phase=Multiply(A=@n3, B=@n4)
    Value 1=ScalarParameter() {defaultValue=0.8, parameterName=Icon Min Size}
    Value 2=ScalarParameter() {defaultValue=1.5, parameterName=Icon Max Size}
   ) {materialFunction=/Engine/Functions/Engine_MaterialFunctions02/Utility/Sine_Remapped}.Remapped Sine
   Alpha=ScalarParameter() {parameterName=bAnimateIcon}
  )
  n6 = TextureSampleParameter2D(UVs=MaterialFunctionCall(UVs=@n1.Normalized UV, Texture Scale=Multiply(A=@n2, B=@n5)) {materialFunction=&ScaleUVsByCenter}.UVs) {parameterName=Center Icon, texture=/Game/UI/Foundation/Textures/Icons/Locks/T-Icon-Lock-64}
  n7 = ScalarParameter() {defaultValue=1, parameterName=Center Icon Alpha}
  n8 = TextureSample(UVs=MaterialFunctionCall() {materialFunction=&GetUserInterfaceUV}.9-Slice UV) {samplerType=SAMPLERTYPE_LinearColor, texture=/Game/UI/Foundation/Textures/Buttons/T-Slot-SelectSquare}
  n9 = ScalarParameter() {parameterName=Darkening Alpha, sliderMax=1}
  n10 = MaterialFunctionCall() {materialFunction=&GetUserInterfaceUV}
  n11 = TextureSampleParameter2D(UVs=@n10.9-Slice UV) {parameterName=Box, samplerType=SAMPLERTYPE_Masks, texture=/Game/UI/Foundation/Textures/Buttons/T-Slot-EmptySquare}
  n12 = Add(A=@n11.R, B=Add(A=@n11.G, B=@n11.G))
  n13 = MaterialFunctionCall() {materialFunction=&GetUserInterfaceUV}
  n14 = Multiply(
   A=Divide(A=@n13.Pixel Size, B=ScalarParameter() {defaultValue=16, desc=Dash Texture Pixel Size, parameterName=Dash Texture Size - Pixels})
   B=@n13.Normalized UV
  )
  n15 = ScalarParameter() {parameterName=Speed}
  n16 = Panner(Coordinate=@n14, Speed=@n15) {speedX=1, speedY=1}
  n17 = AppendVector(A=ScalarParameter() {parameterName=DashShift_Top-Bottom}, B=ScalarParameter() {parameterName=DashShift_Right-Left})
  n18 = TextureSampleParameter2D(UVs=Add(A=Frac(None=@n16), B=@n17)) {parameterName=Dash Texture, samplerType=SAMPLERTYPE_LinearColor, texture=/Game/UI/Foundation/Textures/Pattern/T-Pattern-EmptyMarqueeU}
  n19 = Panner(Coordinate=@n14, Speed=Multiply(A=@n15) {constB=-1}) {speedX=-1, speedY=-1}
  n20 = TextureSampleParameter2D(UVs=Add(A=Frac(None=@n19), B=@n17)) {parameterName=Dash Texture, samplerType=SAMPLERTYPE_LinearColor, texture=&T-Pattern-EmptyMarqueeU}
  n21 = Floor(None=Add(A=@n10.9-Slice UV) {constB=0.5})
  n22 = ComponentMask(None=@n21) {r=true}
  n23 = ComponentMask(None=@n21) {g=true}
  n24 = Add(
   A=LinearInterpolate(B=LinearInterpolate(A=@n18.R, B=@n20.R, Alpha=@n22), Alpha=@n11.B)
   B=LinearInterpolate(B=LinearInterpolate(A=@n20.G, B=@n18.G, Alpha=@n23), Alpha=@n11.G)
  )
  n25 = Constant() {r=1}
  n26 = Subtract(
   A=MaterialFunctionCall(
    UVs=TextureCoordinate()
    CenterPosition=Constant2Vector() {g=0.5, r=0.5}
    Radius=ScalarParameter() {defaultValue=0.5, parameterName=Radius-Outer}
    Density=ScalarParameter() {defaultValue=96, parameterName=Density-Outer}
   ) {materialFunction=&RadialGradientExponential}.RadialGradientExponential
   B=MaterialFunctionCall(
    UVs=TextureCoordinate()
    CenterPosition=Constant2Vector() {g=0.5, r=0.5}
    Radius=ScalarParameter() {defaultValue=0.47, parameterName=Radius-Inner}
    Density=ScalarParameter() {defaultValue=96, parameterName=Density-Inner}
   ) {materialFunction=&RadialGradientExponential}.RadialGradientExponential
  )
  n27 = StaticSwitchParameter(True=Rotator() {speed=-0.5}, False=TextureCoordinate()) {defaultValue=true, parameterName=Rotate}
  n28 = Clamp(None=Subtract(
    A=@n26
    B=TextureSampleParameter2D(UVs=@n27) {parameterName=Dash Hole Texture, texture=/Game/UI/Foundation/Textures/Buttons/T-Slot-GadgetDash-128}.RGB
   ))
  n29 = TextureSampleParameter2D() {parameterName=Banner, texture=/Game/UI/Foundation/Textures/Buttons/T-Slot-Banner-Personal-L-256}
  n30 = StaticSwitchParameter(
   True=StaticSwitchParameter(True=Clamp(None=@n12), False=LinearInterpolate(A=@n24, B=@n25, Alpha=@n11.R)) {parameterName=Solid Line}
   False=StaticSwitchParameter(
    True=StaticSwitchParameter(True=@n28, False=@n26) {defaultValue=true, parameterName=Dash?}
    False=StaticComponentMaskParameter(None=@n29.RGB) {defaultR=true, parameterName=R-Solid G-Dash}
   ) {defaultValue=true, parameterName=Gadget?}
  ) {defaultValue=true, parameterName=Item Card}
  n31 = Multiply(A=ScalarParameter() {defaultValue=1, parameterName=Alpha}, B=Multiply(A=@n8.RGB, B=Add(A=@n9, B=@n30)))
  n32 = Multiply(A=Time(), B=ScalarParameter() {defaultValue=1, parameterName=Pulse Speed})
  n33 = Constant() {r=0.4}
  n34 = Constant() {r=1}
  n35 = StaticSwitchParameter(
   True=Multiply(
    A=@n31
    B=MaterialFunctionCall(Sine Phase=@n32, Value 1=@n33, Value 2=@n34) {materialFunction=&Sine_Remapped}.Remapped Sine
   )
   False=@n31
  ) {parameterName=Animate Alpha}
  StaticSwitchParameter(True=Add(A=Multiply(A=@n6.A, B=@n7), B=@n35), False=@n35) {parameterName=Use Center Icon}
  n36 = VectorParameter() {parameterName=Color1, defaultValue=<rgba 1 1 1 1>}
  n37 = Multiply(A=@n36.RGB, B=@n30)
  n38 = OneMinus(None=@n8.RGB)
  n39 = Multiply(A=VectorParameter() {parameterName=Darkening, defaultValue=<rgba 0 0.004 0.0723 1>}.RGB, B=OneMinus(None=@n30))
  StaticSwitchParameter(True=Add(A=Add(A=@n37, B=@n38), B=Multiply(A=@n8.RGB, B=@n39)), False=@n36.RGB) {defaultValue=true, parameterName=Use Dark Overlay}
 ]
}
