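# Material function in the UI foundation set that builds a procedural triangle mask from three vertex inputs in UV space.
# AspectRatio scales the vertices about center, and XYZ Offset's XY shifts the sampled UVs.
# Each edge's perpendicular, with X weighted by a scalar input, is dotted with the vertex-to-UV offset and clamped.
# The maximum of these edge distances feeds a SmoothStep between SS Min and SS Max.
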MaterialFunction /Game/UI/Foundation/Materials/Mf-Cut
graph {nodes [
  n1 = FunctionInput() {bUsePreviewValueAsDefault=true, inputName=AspectRatio, sortPriority=3, previewValue=<1 1 0 1>}
  n2 = ComponentMask(None=Multiply(A=@n1)) {g=true, r=true}
  n3 = ComponentMask(None=MaterialFunctionCall(UVs=FunctionInput() {bUsePreviewValueAsDefault=true, inputName=Vert 2, sortPriority=1, previewValue=<0.6 0.4 0 1>}, Texture Scale=@n2) {materialFunction=&ScaleUVsByCenter}.UVs) {g=true, r=true}
  n4 = ComponentMask(None=MaterialFunctionCall(UVs=FunctionInput() {bUsePreviewValueAsDefault=true, inputName=Vert 1, previewValue=<0.4 0.4 0 1>}, Texture Scale=@n2) {materialFunction=&ScaleUVsByCenter}.UVs) {g=true, r=true}
  n5 = MaterialFunctionCall(Float2=Normalize(VectorInput=Subtract(A=@n3, B=@n4))) {materialFunction=&BreakOutFloat2Components}
  n6 = FunctionInput() {bUsePreviewValueAsDefault=true, inputType=FunctionInput_Scalar, sortPriority=5, previewValue=<-0.9472 0 0 1>}
  n7 = MaterialFunctionCall(X=Multiply(A=@n5.G, B=@n6), Y=@n5.R) {materialFunction=&MakeFloat2}
  n8 = FunctionInput() {bUsePreviewValueAsDefault=true, inputName=XYZ Offset, sortPriority=4}
  n9 = Multiply(A=Add(A=TextureCoordinate(), B=ComponentMask(None=@n8) {g=true, r=true}), B=Constant2Vector() {g=1, r=1})
  n10 = Subtract(A=@n4, B=@n9)
  n11 = FunctionInput() {bUsePreviewValueAsDefault=true, inputName=Vert 3, sortPriority=2, previewValue=<0.5 0.55 0 1>}
  n12 = ComponentMask(None=MaterialFunctionCall(UVs=@n11, Texture Scale=@n2) {materialFunction=&ScaleUVsByCenter}.UVs) {g=true, r=true}
  n13 = MaterialFunctionCall(Float2=Normalize(VectorInput=Subtract(A=@n12, B=@n3))) {materialFunction=&BreakOutFloat2Components}
  n14 = MaterialFunctionCall(X=Multiply(A=@n6, B=@n13.G), Y=@n13.R) {materialFunction=&MakeFloat2}
  n15 = Subtract(A=@n3, B=@n9)
  n16 = Max(A=Clamp(None=DotProduct(A=@n7.Result, B=@n10)), B=Clamp(None=DotProduct(A=@n14.Result, B=@n15)))
  n17 = MaterialFunctionCall(Float2=Normalize(VectorInput=Subtract(A=@n4, B=@n12))) {materialFunction=&BreakOutFloat2Components}
  n18 = Multiply(A=@n6, B=@n17.G)
  n19 = Clamp(None=DotProduct(
    A=MaterialFunctionCall(X=@n18, Y=@n17.R) {materialFunction=&MakeFloat2}.Result
    B=Subtract(A=@n12, B=@n9)
   ))
  FunctionOutput(None=MaterialFunctionCall(
    Alpha=Max(A=@n16, B=@n19)
    Min=FunctionInput() {inputName=SS Min, inputType=FunctionInput_Scalar, previewValue=<0.125 0 0 1>}
    Max=FunctionInput() {inputName=SS Max, inputType=FunctionInput_Scalar, previewValue=<0.15 0 0 1>}
   ) {materialFunction=&SmoothStep}.Result)
 ]}
