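# Translucent UI master material for headers, building a folder-tab shape with a configurable color fill.
# The shape mask combines offset and rotated UV edges via Min/Max and SmoothStep, optionally multiplied by one or two three-vertex Mf-Cut masks.
# Static switches select solid color or a linear, reflected or inverted two-color gradient, with a darken control.
# No material output links are shown, and several tab-angle and top-gap branches end unconsumed.
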Material /Game/UI/Foundation/Materials/M-Header-Master
graph {
 properties {
  bUsedWithStaticMesh=false
  blendMode=BLEND_Translucent
  materialDomain=MD_UI
  pixelDepthOffsetMode=PDOM_Legacy
  shadingModel=MSM_Unlit
  usedShadingModels=Unlit
 }
 nodes [
  n1 = Subtract(
   A=TextureCoordinate()
   B=AppendVector(A=ScalarParameter() {defaultValue=-0.12, parameterName=Offset U_1}, B=ScalarParameter() {defaultValue=-0.12, parameterName=Offset V_1})
  )
  n2 = MaterialFunctionCall(Float2=@n1) {materialFunction=&BreakOutFloat2Components}
  Add(A=Multiply(A=ScalarParameter() {defaultValue=-1.5, parameterName=Tab Angle_1}, B=@n2.R), B=@n2.G)
  n3 = Subtract(A=TextureCoordinate(), B=ScalarParameter() {defaultValue=0.1, parameterName=Folder Top})
  n4 = ScalarParameter() {defaultValue=-1.5, parameterName=Tab Angle_2}
  n5 = Add(A=ComponentMask(None=Rotator(Coordinate=@n1, Time=@n4)) {r=true}, B=Constant2Vector() {r=-0.22})
  n6 = ComponentMask(None=Subtract(A=TextureCoordinate(), B=ScalarParameter() {defaultValue=0.6, parameterName=Folder Button})) {g=true}
  n7 = ScalarParameter() {defaultValue=0.125, parameterName=SS Min}
  n8 = ScalarParameter() {defaultValue=0.15, parameterName=SS Max}
  n9 = MaterialFunctionCall(Alpha=Min(A=ComponentMask(None=@n3) {g=true}, B=Max(A=@n5, B=@n6)), Min=@n7, Max=@n8) {materialFunction=&SmoothStep}
  n10 = ComponentMask(None=Subtract(A=TextureCoordinate(), B=ScalarParameter() {defaultValue=-0.08, parameterName=Box Offset}) {constB=0.0358}) {g=true}
  n11 = StaticSwitchParameter(True=@n10, False=OneMinus(None=@n10)) {defaultValue=true, parameterName=Top Gap}
  Add(
   A=@n9.Result
   B=MaterialFunctionCall(Alpha=Max(A=@n11) {constB=0}, Min=@n7, Max=@n8) {materialFunction=&SmoothStep}.Result
  )
  n12 = VectorParameter() {group=Color, parameterName=RGBA0, defaultValue=<rgba 1 0 0 1>}
  n13 = VectorParameter() {group=Color, parameterName=RGBA1, defaultValue=<rgba 0 1 0 1>}
  n14 = MaterialFunctionCall() {materialFunction=&GetUserInterfaceUV}
  n15 = Abs(None=Add(A=@n14.Normalized UV) {constB=-0.5})
  n16 = ScalarParameter() {defaultValue=1, group=Linear Gradient Settings, parameterName=Rotation}
  n17 = Divide(
   A=StaticComponentMaskParameter(None=Rotator(Coordinate=@n14.Normalized UV, Time=@n16)) {defaultR=true, group=Options, parameterName=Gradient Direction}
   B=ScalarParameter() {defaultValue=1, group=Linear Gradient Settings, parameterName=Scale_1}
  )
  n18 = ScalarParameter() {group=Linear Gradient Settings, parameterName=Position}
  n19 = StaticSwitchParameter(
   True=MaterialFunctionCall(
    Alpha=StaticComponentMaskParameter(None=@n15) {defaultR=true, group=Options, parameterName=Reflected Gradient Direction}
    Min=ScalarParameter() {group=Reflected Gradient Settings, parameterName=RG SS Min}
    Max=ScalarParameter() {defaultValue=1, group=Reflected Gradient Settings, parameterName=RG SS Max}
   ) {materialFunction=&SmoothStep}.Result
   False=MaterialFunctionCall(
    Alpha=Add(A=@n17, B=@n18)
    Min=ScalarParameter() {group=Linear Gradient Settings, parameterName=LG SS Min}
    Max=ScalarParameter() {defaultValue=1, group=Linear Gradient Settings, parameterName=LG SS Max}
   ) {materialFunction=&SmoothStep}.Result
  ) {group=Options, parameterName=Reflected Linear}
  n20 = StaticSwitchParameter(True=OneMinus(None=@n19), False=@n19) {group=Options, parameterName=Invert Gradient}
  LinearInterpolate(
   A=StaticSwitchParameter(True=@n12.RGB, False=LinearInterpolate(A=@n12.RGB, B=@n13.RGB, Alpha=@n20)) {parameterName=Solid Color}
   Alpha=ScalarParameter() {parameterName=Darken RGB}
  ) {constB=0}
  n21 = MaterialFunctionCall!(
   SS Min=ScalarParameter() {defaultValue=0.125, parameterName=SS Min Cut}
   SS Max=ScalarParameter() {defaultValue=0.15, parameterName=SS Max Cut}
   Vert 1=VectorParameter() {parameterName=Cut Vert1, defaultValue=<rgba 0.4 0.4 0 0>}.RGB
   Vert 2=VectorParameter() {parameterName=Cut Vert2, defaultValue=<rgba 0.6 0.4 0 0>}.RGB
   Vert 3=VectorParameter() {parameterName=Cut Vert3, defaultValue=<rgba 0.5 0.55 0 0>}.RGB
   AspectRatio=VectorParameter() {parameterName=Cut Aspect Ratio, defaultValue=<rgba 1 2 0 0>}.RGB
   XYZ Offset=VectorParameter() {parameterName=Cut Offset, defaultValue=<rgba -0.23 0.37 0 0>}.RGB
   In=ScalarParameter() {defaultValue=3, parameterName=Cut scale}
  ) {materialFunction=&/Mf-Cut}
  n22 = Multiply(
   A=@n21.Result
   B=MaterialFunctionCall(
    SS Min=ScalarParameter() {defaultValue=0.125, parameterName=SS Min Cut2}
    SS Max=ScalarParameter() {defaultValue=0.15, parameterName=SS Max Cut2}
    Vert 1=VectorParameter() {parameterName=Cut2 Vert1, defaultValue=<rgba 0.4 0.4 0 0>}.RGB
    Vert 2=VectorParameter() {parameterName=Cut2 Vert2, defaultValue=<rgba 0.6 0.4 0 0>}.RGB
    Vert 3=VectorParameter() {parameterName=Cut2 Vert3, defaultValue=<rgba 0.5 0.55 0 0>}.RGB
    AspectRatio=VectorParameter() {parameterName=Cut2 Aspect Ratio, defaultValue=<rgba 1 2 0 0>}.RGB
    XYZ Offset=VectorParameter() {parameterName=Cut2 Offset, defaultValue=<rgba -0.23 0.37 0 0>}.RGB
    In=ScalarParameter() {defaultValue=3, parameterName=Cut2 scale}
   ) {materialFunction=&Mf-Cut}.Result
  )
  Multiply(
   A=@n9.Result
   B=StaticSwitchParameter(True=StaticSwitchParameter(True=@n22, False=@n21.Result) {parameterName=Cut 2}, False=Constant() {r=1}) {parameterName=Cut 1}
  )
  ScalarParameter() {defaultValue=0.5, parameterName=Offset}
  n23 = ScalarParameter() {defaultValue=-0.12, parameterName=Offset U}
  n24 = ScalarParameter() {defaultValue=-0.12, parameterName=Offset V}
  n25 = MaterialFunctionCall(Float2=Subtract(A=TextureCoordinate(), B=AppendVector(A=@n23, B=@n24))) {materialFunction=&BreakOutFloat2Components}
  n26 = Add(A=Multiply(A=ScalarParameter() {defaultValue=-0.0365, parameterName=Tab Angle}, B=@n25.R), B=@n25.G)
  StaticSwitchParameter(True=@n26, False=OneMinus(None=@n26)) {defaultValue=true, parameterName=Top Gap}
 ]
}
