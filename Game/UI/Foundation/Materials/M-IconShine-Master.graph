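# Translucent UI master material that adds a shine swipe over a parameterized icon texture.
# The icon RGB lerps toward an optional tint color, then Color Dodge blends it with the M-Function-ShineSwipe result.
# A separate inline chain, not feeding that blend, pans a rotated shine texture and gates it with a thresholded sine of time for intermittent flashes.
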Material /Game/UI/Foundation/Materials/M-IconShine-Master
graph {
 properties {
  bUsedWithStaticMesh=false
  blendMode=BLEND_Translucent
  materialDomain=MD_UI
  pixelDepthOffsetMode=PDOM_Legacy
  shadingModel=MSM_Unlit
  usedShadingModels=Unlit
 }
 nodes [
  MaterialFunctionCall!(
   Blend=MaterialFunctionCall!(In_Uv) {materialFunction=&/M-Function-ShineSwipe}.Result
   Base=LinearInterpolate(
    A=TextureSampleParameter2D() {parameterName=IconTexture, texture=/Game/Athena/UI/Frontend/Spatial/Art/T_Ui_Empty}.RGB
    B=VectorParameter() {parameterName=Tint Color, defaultValue=<rgba 1 0 0.9925 0>}.RGB
    Alpha=ScalarParameter() {parameterName="Tint Icon (False)", sliderMax=1}
   )
  ) {materialFunction=/Engine/Functions/Engine_MaterialFunctions03/Blends/Blend_ColorDodge}
  n1 = Multiply(A=Time(), B=ScalarParameter() {defaultValue=0.25, parameterName=Shine Time_1})
  n2 = TextureCoordinate() {coordinateIndex=4, uTiling=0.4, vTiling=0.2}
  n3 = Constant() {r=1.9}
  n4 = Time()
  n5 = ScalarParameter() {defaultValue=1, parameterName=Shine Time}
  n6 = Frac(None=Panner(Coordinate=Rotator(Coordinate=@n2, Time=@n3), Time=Multiply(A=@n4, B=@n5)) {constCoordinate=4, speedX=-2, speedY=-1})
  Multiply(
   A=If(A=Sine(None=@n1), B=Constant() {r=0.9}, "A > B"=Constant() {r=1}, "A < B"=Constant())
   B=Multiply(
    A=TextureSample(UVs=@n6) {texture=/Game/UI/Foundation/Textures/Pattern/T-UI-Shine}.RGB
    B=ScalarParameter() {defaultValue=0.75, parameterName=Shine Alpha, sliderMax=1}
   )
  )
 ]
}
