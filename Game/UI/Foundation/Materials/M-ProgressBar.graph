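# UI-domain unlit material for a segmentable progress bar with primary, secondary, tertiary and negative fills over a background color.
# Floored UV-plus-value thresholds, with selectable axis and reversal, stack secondary and tertiary after primary and carve negative from primary's end.
# Mf-UILinearGradientColor colors each fill; both inputs of every "Reflected" switch use the reflected output.
# Static switches add split lines from repeated UVs and choose a 9-slice texture or procedural edge mask.
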Material /Game/UI/Foundation/Materials/M-ProgressBar
graph {
 properties {bUsedWithStaticMesh=false, materialDomain=MD_UI, pixelDepthOffsetMode=PDOM_Legacy, shadingModel=MSM_Unlit, usedShadingModels=Unlit}
 nodes [
  n1 = TextureCoordinate() {coordinateIndex=4}
  n2 = ScalarParameter() {defaultValue=4, parameterName=splits}
  n3 = Frac(None=Add(A=Multiply(A=@n1, B=@n2)) {constB=-0.5})
  n4 = Subtract(A=Multiply(A=@n3, B=OneMinus(None=@n3)), B=ScalarParameter() {defaultValue=0.249, parameterName=Line Thickness, sliderMax=1}) {constB=12}
  Ceil(None=@n4)
  n5 = Frac(None=TextureCoordinate() {coordinateIndex=4})
  ComponentMask(None=@n5) {r=true}
  Multiply()
  ScalarParameter() {parameterName=1 UV Type}
  n6 = MaterialFunctionCall() {materialFunction=&GetUserInterfaceUV}
  n7 = Rotator(Coordinate=TextureCoordinate(), Time=ScalarParameter() {defaultValue=0.55, parameterName=Rotation})
  n8 = Subtract(A=Abs(None=Add(A=@n7) {constB=-0.5}), B=ScalarParameter() {parameterName=Edge Power}) {constB=-0.35}
  n9 = ScalarParameter() {defaultValue=0.18, group=-Mask, parameterName=SS Min}
  n10 = ScalarParameter() {defaultValue=0.125, group=-Mask, parameterName=SS Max}
  StaticSwitchParameter(
   True=StaticComponentMaskParameter(None=TextureSampleParameter2D(UVs=@n6.9-Slice UV) {group=-Master, parameterName=Mask, texture=/Game/Global/Textures/Default/Blanks/T_WhiteUserInterface}.RGB) {defaultR=true, group=-Master, parameterName=Mask Channel}
   False=Multiply(
    A=MaterialFunctionCall(Alpha=@n8, Min=@n9, Max=@n10) {materialFunction=&SmoothStep}.Result
    B=ScalarParameter() {defaultValue=1, group=-Master, parameterName=Alpha, sliderMax=1}
   )
  ) {group=-Master, parameterName=Box Scale Mask}
  n11 = StaticComponentMaskParameter(None=@n5) {defaultR=true, parameterName=BarDirection}
  n12 = Add(
   A=LinearInterpolate(A=OneMinus(None=@n11), B=@n11, Alpha=ScalarParameter() {parameterName=bReverseDireciton})
   B=ScalarParameter() {defaultValue=0.5, group=1 Primary, parameterName=Bar1}
  ) {constB=0.3}
  n13 = Floor(None=@n12)
  n14 = MaterialFunctionCall!(
   In
   RGBA0=VectorParameter() {group=1 Primary, parameterName=Bar1RGBA0, defaultValue=<rgba 1 1 1 1>}.RGB
   RGBA1=VectorParameter() {group=1 Primary, parameterName=Bar1RGBA1, defaultValue=<rgba 1 1 1 1>}.RGB
   SS Min=ScalarParameter() {group=1 Primary, parameterName=SS Min 1}
   SS Max=ScalarParameter() {defaultValue=1, group=1 Primary, parameterName=SS Max 1}
   Gradient Rotation=ScalarParameter() {group=1 Primary, parameterName=Rotation 1}
  ) {materialFunction=&/Mf-UILinearGradientColor}
  n15 = LinearInterpolate(
   A=VectorParameter() {group=-Master, parameterName=Background, defaultValue=<rgba 0.12 0.12 0.12 1>}.RGB
   B=StaticSwitchParameter(True=@n14.Reflected Gradient, False=@n14.Reflected Gradient) {group=1 Primary, parameterName=Bar 1 Reflected}
   Alpha=Clamp(None=@n13)
  ) {desc=Bar 1}
  n16 = VectorParameter() {group=2 Secondary, parameterName=Bar2RGBA0, defaultValue=<rgba 0.795 0.39 0.3265 1>}
  n17 = OneMinus(None=@n13)
  n18 = Add(A=@n12, B=ScalarParameter() {defaultValue=0.1, group=2 Secondary, parameterName=Bar2})
  n19 = Floor(None=@n18)
  n20 = Multiply(A=@n16.A, B=Clamp(None=Min(A=@n17, B=@n19)))
  n21 = MaterialFunctionCall(
   RGBA0=@n16.RGB
   RGBA1=VectorParameter() {group=2 Secondary, parameterName=Bar2RGBA1, defaultValue=<rgba 0.7913 0.3813 0.3185 1>}.RGB
   SS Min=ScalarParameter() {group=2 Secondary, parameterName=SS Min 2}
   SS Max=ScalarParameter() {defaultValue=1, group=2 Secondary, parameterName=SS Max 2}
   Gradient Rotation=ScalarParameter() {defaultValue=0.5012, group=2 Secondary, parameterName=Rotation 2}
  ) {materialFunction=&Mf-UILinearGradientColor}
  n22 = StaticSwitchParameter(True=@n21.Reflected Gradient, False=@n21.Reflected Gradient) {group=2 Secondary, parameterName=Bar 2 Reflected}
  n23 = VectorParameter() {group=3 Tertiary, parameterName=Bar3RGBA0, defaultValue=<rgba 0.3007 0.3782 0.735 1>}
  n24 = Add(A=@n18, B=ScalarParameter() {defaultValue=0.1, group=3 Tertiary, parameterName=Bar3})
  n25 = Clamp(None=Min(A=OneMinus(None=@n19), B=Floor(None=@n24)))
  n26 = MaterialFunctionCall(
   RGBA0=@n23.RGB
   RGBA1=VectorParameter() {group=3 Tertiary, parameterName=Bar3RGBA1, defaultValue=<rgba 0.2961 0.3712 0.7305 1>}.RGB
   SS Min=ScalarParameter() {group=3 Tertiary, parameterName=SS Min 3}
   SS Max=ScalarParameter() {defaultValue=1, group=3 Tertiary, parameterName=SS Max 3}
   Gradient Rotation=ScalarParameter() {group=3 Tertiary, parameterName=Rotation 3}
  ) {materialFunction=&Mf-UILinearGradientColor}
  n27 = VectorParameter() {group=4 Negative, parameterName=NegativeBar4RGBA0, defaultValue=<rgba 1 0 0 1>}
  n28 = Subtract(A=@n12, B=ScalarParameter() {group=4 Negative, parameterName=NegativeBar4})
  n29 = Min(A=@n13, B=OneMinus(None=Floor(None=@n28)))
  n30 = MaterialFunctionCall(
   RGBA0=@n27.RGB
   RGBA1=VectorParameter() {group=4 Negative, parameterName=NegativeBar4RGBA1, defaultValue=<rgba 1 0 0 1>}.RGB
   SS Min=ScalarParameter() {group=4 Negative, parameterName=SS Min 4}
   SS Max=ScalarParameter() {defaultValue=1, group=4 Negative, parameterName=SS Max 4}
   Gradient Rotation=ScalarParameter() {group=4 Negative, parameterName=Rotation 4}
  ) {materialFunction=&Mf-UILinearGradientColor}
  n31 = LinearInterpolate(
   A=LinearInterpolate(
    A=LinearInterpolate(A=@n15, B=@n22, Alpha=@n20) {desc=Bar 2}
    B=StaticSwitchParameter(True=@n26.Reflected Gradient, False=@n26.Reflected Gradient) {group=3 Tertiary, parameterName=Bar 3 Reflected}
    Alpha=Multiply(A=@n23.A, B=@n25)
   ) {desc=Bar3}
   B=StaticSwitchParameter(True=@n30.Reflected Gradient, False=@n30.Reflected Gradient) {group=4 Negative, parameterName=Bar 4 Reflected}
   Alpha=Multiply(A=@n27.A, B=Clamp(None=@n29))
  ) {desc=Negative Bar}
  n32 = VectorParameter() {parameterName=Split Color, defaultValue=<rgba 1 0 0.8585 0.1>}
  n33 = TextureCoordinate()
  n34 = Subtract(A=Abs(None=Add(A=@n33) {constB=-0.5})) {constB=0.45}
  n35 = OneMinus(None=Ceil(None=ComponentMask(None=@n34) {r=true}))
  n36 = Constant() {r=8}
  n37 = ComponentMask(None=MaterialFunctionCall(Alpha=@n4, Min=Constant(), Max=Divide(A=@n36) {constB=10000}) {materialFunction=&SmoothStep}.Result) {r=true}
  n38 = Multiply(A=Clamp(None=Multiply(A=@n35, B=@n37)), B=@n32.A)
  StaticSwitchParameter(True=LinearInterpolate(A=@n31, B=@n32.RGB, Alpha=Clamp(None=@n38)), False=@n31) {desc=Split bar into chunks, parameterName=Segments}
 ]
}
