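# Translucent UI material for the shell top bar, drawing a two-tone shape with an angled edge tab.
# It builds a mask from the max of a vertical offset band and a rotated UV gradient, which Switch Side blends to its inverted version.
# A narrow SmoothStep sharpens the mask, which then blends inner and outer colors and their separate alphas.
# The Solid static switch replaces this with the outer color at full alpha.
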Material /Game/UI/Foundation/Materials/M-Shell-TopBar
graph {
 properties {
  bUsedWithStaticMesh=false
  blendMode=BLEND_Translucent
  materialDomain=MD_UI
  pixelDepthOffsetMode=PDOM_Legacy
  shadingModel=MSM_Unlit
  usedShadingModels=Unlit
 }
 nodes [
  n1 = TextureCoordinate()
  n2 = OneMinus(None=ComponentMask(None=@n1) {g=true})
  n3 = ScalarParameter() {defaultValue=0.4575, parameterName=Top Offset}
  n4 = Rotator(Coordinate=@n1, Time=ScalarParameter() {defaultValue=1, parameterName=Rotation})
  n5 = ScalarParameter() {defaultValue=0.2182, parameterName=Edge Tab}
  n6 = ScalarParameter() {parameterName=Switch Side, sliderMax=1}
  n7 = Multiply(A=Subtract(A=@n4, B=@n5) {constB=0.42}, B=OneMinus(None=@n6))
  n8 = Multiply(A=Subtract(A=OneMinus(None=@n4), B=@n5) {constB=0.42}, B=@n6)
  n9 = MaterialFunctionCall(Alpha=Max(A=Subtract(A=@n2, B=@n3) {constB=0.1}, B=Add(A=@n7, B=@n8)), Min=Constant() {r=0.465}, Max=Constant() {r=0.5}) {materialFunction=&SmoothStep}
  n10 = Multiply(A=OneMinus(None=@n9.Result), B=ScalarParameter() {defaultValue=0.5, parameterName=PowerBG Alpha})
  n11 = Multiply(A=@n9.Result, B=ScalarParameter() {defaultValue=1, parameterName=Outer Edge Alpha})
  StaticSwitchParameter(True=Constant() {r=1}, False=Add(A=Clamp(None=@n10), B=Clamp(None=@n11))) {parameterName=Solid}
  n12 = VectorParameter() {parameterName=Outer Color, defaultValue=<rgba 0.0116 0.016 0.0395 1>}
  StaticSwitchParameter(
   True=@n12.RGB
   False=LinearInterpolate(A=VectorParameter() {parameterName=Inner Color, defaultValue=<rgba 0 0.0021 0.0176 1>}.RGB, B=@n12.RGB, Alpha=@n9.Result)
  ) {parameterName=Solid}
 ]
}
