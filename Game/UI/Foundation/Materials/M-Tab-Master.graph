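# Translucent unlit UI master material that draws a tab or folder shape with a solid or two-color gradient fill.
# The gradient uses screen or widget UVs and switches between rotated linear and reflected smoothstep ramps, optionally inverted, blending both RGB and alpha.
# An optional smoothstep mask cuts the shape from UV thresholds, with a rotated edge forming the slanted tab.
# Alpha combines fill alpha, that mask and an overall Alpha parameter.
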Material /Game/UI/Foundation/Materials/M-Tab-Master
graph {
 properties {
  bUsedWithStaticMesh=false
  blendMode=BLEND_Translucent
  materialDomain=MD_UI
  pixelDepthOffsetMode=PDOM_Legacy
  shadingModel=MSM_Unlit
  usedShadingModels=Unlit
 }
 nodes [
  n1 = Subtract(A=TextureCoordinate(), B=ScalarParameter() {defaultValue=0.5, parameterName=Offset})
  n2 = MaterialFunctionCall(Float2=@n1) {materialFunction=&BreakOutFloat2Components}
  Add(A=Multiply(A=ScalarParameter() {defaultValue=-1.5, parameterName=Tab Angle}, B=@n2.R), B=@n2.G)
  n3 = VectorParameter() {group=Color, parameterName=RGBA0, defaultValue=<rgba 1 0 0 1>}
  n4 = VectorParameter() {group=Color, parameterName=RGBA1, defaultValue=<rgba 0 1 0 1>}
  n5 = StaticSwitchParameter(
   True=Multiply(A=Constant() {r=1}, B=ScreenPosition().ViewportUV)
   False=MaterialFunctionCall() {materialFunction=&GetUserInterfaceUV}.Normalized UV
  ) {parameterName=Screenspace Gradient}
  n6 = Abs(None=Add(A=@n5) {constB=-0.5})
  n7 = ScalarParameter() {defaultValue=1, group=Linear Gradient Settings, parameterName=Rotation}
  n8 = Divide(
   A=StaticComponentMaskParameter(None=Rotator(Coordinate=@n5, Time=@n7)) {defaultR=true, group=Options, parameterName=Gradient Direction}
   B=ScalarParameter() {defaultValue=1, group=Linear Gradient Settings, parameterName=Scale_1}
  )
  n9 = ScalarParameter() {group=Linear Gradient Settings, parameterName=Position}
  n10 = StaticSwitchParameter(
   True=MaterialFunctionCall(
    Alpha=StaticComponentMaskParameter(None=@n6) {defaultR=true, group=Options, parameterName=Reflected Gradient Direction}
    Min=ScalarParameter() {group=Reflected Gradient Settings, parameterName=RG SS Min}
    Max=ScalarParameter() {defaultValue=1, group=Reflected Gradient Settings, parameterName=RG SS Max}
   ) {materialFunction=&SmoothStep}.Result
   False=MaterialFunctionCall(
    Alpha=Add(A=@n8, B=@n9)
    Min=ScalarParameter() {group=Linear Gradient Settings, parameterName=LG SS Min}
    Max=ScalarParameter() {defaultValue=1, group=Linear Gradient Settings, parameterName=LG SS Max}
   ) {materialFunction=&SmoothStep}.Result
  ) {group=Options, parameterName=Reflected Linear}
  n11 = StaticSwitchParameter(True=OneMinus(None=@n10), False=@n10) {group=Options, parameterName=Invert Gradient}
  ComponentMask(None=StaticSwitchParameter(True=@n3.RGB, False=LinearInterpolate(A=@n3.RGB, B=@n4.RGB, Alpha=@n11)) {parameterName=Solid Color}) {a=true}
  Constant() {r=1}
  MaterialFunctionCall!(Sphere Radius, Sphere Falloff, Unit Ray Direction, Sphere Center, Ray Origin) {materialFunction=/Engine/Functions/Engine_MaterialFunctions02/Gradients/RayTraceSphereFalloff}
  n12 = StaticSwitchParameter(True=@n3.A, False=LinearInterpolate(A=@n3.A, B=@n4.A, Alpha=@n11)) {parameterName=Solid Color}
  n13 = ScalarParameter() {defaultValue=1, parameterName=Alpha, sliderMax=1}
  n14 = ComponentMask(None=Subtract(A=TextureCoordinate(), B=ScalarParameter() {defaultValue=0.1, parameterName=Folder Top})) {g=true}
  n15 = Rotator(Coordinate=@n1, Time=ScalarParameter() {defaultValue=-1.5, parameterName=Tab Angle})
  n16 = TextureCoordinate()
  n17 = ScalarParameter() {defaultValue=0.6, parameterName=Folder Button}
  n18 = Max(A=Add(A=ComponentMask(None=@n15) {r=true}, B=Constant2Vector() {r=-0.22}), B=ComponentMask(None=Subtract(A=@n16, B=@n17)) {g=true})
  n19 = MaterialFunctionCall(
   Alpha=StaticSwitchParameter(True=Min(A=@n14, B=@n18), False=@n14) {defaultValue=true, parameterName=Tab}
   Min=ScalarParameter() {defaultValue=0.18, parameterName=SS Min}
   Max=ScalarParameter() {defaultValue=0.16, parameterName=SS Max}
  ) {materialFunction=&SmoothStep}
  StaticSwitchParameter(True=Multiply(A=@n12, B=@n13), False=Multiply(A=Multiply(A=@n12, B=@n19.Result), B=@n13)) {parameterName=No Shape Mask}
 ]
}
