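# Translucent UI material that draws an angled fill panel with separate inner and outline colors and opacities, apparently for the top bar.
# Rotated, offset UVs form slanted left and right edges, smoothstepped to clip opacity and combined with a top/bottom band into the outline mask.
# Split parameters let the left edge's rotation and offset differ from the shared values.
# A static switch optionally adds a shine swipe from a material function, confined to the outline.
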Material /Game/UI/Foundation/Materials/M-TopBar-AngleFills
graph {
 properties {
  bUsedWithStaticMesh=false
  blendMode=BLEND_Translucent
  materialDomain=MD_UI
  pixelDepthOffsetMode=PDOM_Legacy
  shadingModel=MSM_Unlit
  usedShadingModels=Unlit
 }
 nodes [
  n1 = VectorParameter() {parameterName=InnerRGBA, defaultValue=<rgba 1 0 0.5192 0.4>}
  n2 = VectorParameter() {parameterName=OuterRGBA, defaultValue=<rgba 0 0.4193 1 0.611>}
  n3 = TextureCoordinate()
  n4 = Abs(None=Add(A=@n3) {constB=-0.5})
  n5 = Add(A=ScalarParameter() {defaultValue=0.135, parameterName=TopBottomOutlineThickness}) {constB=-0.5}
  n6 = Multiply(A=ComponentMask(None=Add(A=@n4, B=@n5)) {g=true}) {constB=128}
  n7 = ScalarParameter() {defaultValue=0.04, parameterName=EdgeRotation}
  n8 = Add(A=Rotator(Coordinate=@n3, Time=@n7) {centerX=1, centerY=1, speed=6.2832}) {constB=-1}
  n9 = ScalarParameter() {defaultValue=0.15, parameterName=LeftRightOutlineThickness}
  n10 = Add(A=@n8, B=@n9)
  n11 = ScalarParameter() {defaultValue=0.0225, parameterName=EdgeOffset}
  n12 = Rotator(
   Coordinate=@n3
   Time=LinearInterpolate(A=@n7, B=ScalarParameter() {parameterName=LeftEdgeRotation}, Alpha=ScalarParameter() {parameterName=SplitEdgeRotation})
  ) {centerX=0, centerY=0, speed=6.2832}
  n13 = Add(A=@n12, B=Multiply(A=@n9) {constB=-1})
  n14 = Multiply(A=LinearInterpolate(A=@n11, B=ScalarParameter() {parameterName=LeftEdgeOffset}, Alpha=ScalarParameter() {parameterName=SplitEdgeOffset})) {constB=-1}
  n15 = Multiply(A=ComponentMask(None=Add(A=@n10, B=@n11) {constB=0.1}) {r=true}, B=ComponentMask(None=Add(A=@n13, B=@n14) {constB=-0.1}) {r=true})
  n16 = Constant()
  n17 = ScalarParameter() {defaultValue=0.015, parameterName=SSThreshold}
  n18 = Clamp(None=Add(
    A=Saturate(None=@n6)
    B=MaterialFunctionCall(Alpha=@n15, Min=@n16, Max=@n17) {materialFunction=&SmoothStep}.Result
   ))
  n19 = Add(A=@n8, B=@n11) {constB=0.1}
  n20 = Add(A=@n12, B=@n14) {constB=-0.1}
  n21 = MaterialFunctionCall(Alpha=Multiply(A=ComponentMask(None=@n19) {r=true}, B=ComponentMask(None=@n20) {r=true}), Min=@n16, Max=@n17) {materialFunction=&SmoothStep}
  Multiply(A=LinearInterpolate(A=@n1.A, B=@n2.A, Alpha=@n18), B=Clamp(None=OneMinus(None=@n21.Result)))
  n22 = LinearInterpolate(A=@n1.RGB, B=@n2.RGB, Alpha=@n18)
  n23 = VectorParameter() {parameterName=SwipeRGB, defaultValue=<rgba 0.9734 1 0.4452 1>}
  n24 = MaterialFunctionCall!(In_Uv) {materialFunction=&/M-Function-ShineSwipe}
  StaticSwitchParameter(True=LinearInterpolate(A=@n22, B=LinearInterpolate(A=@n22, B=@n23.RGB, Alpha=@n24.Result), Alpha=@n18), False=@n22) {parameterName=UseShineSwipe}
 ]
}
