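# Translucent UI material that draws an animated, two-color procedural vortex effect.
# A sine-oscillated rotating vortex mask and radially panned coordinates with blurred noise are summed to distort UVs for sampling the same vortex texture.
# Center-scaled copies of its blue channel drive a power and subtraction that shape the alpha, while the powered pattern blends the colors.
# Parameters expose direction, speed, falloff, powers and colors.
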Material /Game/UI/Foundation/Materials/M-UI-VortexFX
graph {
 properties {
  bUsedWithStaticMesh=false
  blendMode=BLEND_Translucent
  materialDomain=MD_UI
  pixelDepthOffsetMode=PDOM_Legacy
  usedShadingModels=Default Lit
 }
 nodes [
  n1 = Time()
  n2 = Sine(None=@n1) {period=8}
  n3 = AppendVector(
   A=TextureSample(UVs=Rotator(Time=@n2) {speed=1}) {samplerType=SAMPLERTYPE_Masks, texture=/Game/UI/Foundation/Textures/Pattern/T-FlowMap-vortex}.R
   B=ComponentMask(None=TextureCoordinate()) {g=true}
  )
  n4 = MaterialFunctionCall() {materialFunction=&VectorToRadialValue}
  n5 = Multiply(A=@n1, B=ScalarParameter() {defaultValue=-1, parameterName=Vortex Direction, sliderMax=1, sliderMin=-1}) {constB=-1}
  n6 = ScalarParameter() {defaultValue=0.25, parameterName=Speed, sliderMax=1, sliderMin=-1}
  n7 = AppendVector(
   A=ComponentMask(None=Panner(Coordinate=@n4.Linear Distance, Time=@n5, Speed=@n6) {speedX=-0.5, speedY=-0.5}) {r=true}
   B=TextureSample(UVs=Panner(Coordinate=@n4.Linear Distance, Time=@n5, Speed=@n6) {speedX=-0.5, speedY=-0.5}) {samplerType=SAMPLERTYPE_Grayscale, texture=&LowResBlurredNoise}.R
  )
  n8 = TextureSample(UVs=MaterialFunctionCall(Texture Scale=Constant() {r=2}) {materialFunction=&ScaleUVsByCenter}.UVs) {samplerType=SAMPLERTYPE_Masks, texture=&T-FlowMap-vortex}
  n9 = Constant() {r=0.75}
  n10 = OneMinus(None=TextureSample(UVs=MaterialFunctionCall(Texture Scale=@n9) {materialFunction=&ScaleUVsByCenter}.UVs) {samplerType=SAMPLERTYPE_Masks, texture=&T-FlowMap-vortex}.B)
  n11 = Power(
   Base=TextureSample(UVs=Add(A=@n3, B=@n7)) {samplerType=SAMPLERTYPE_Masks, texture=&T-FlowMap-vortex}.R
   Exp=Multiply(A=Add(A=@n8.B, B=@n10), B=ScalarParameter() {defaultValue=0.95, parameterName=OpacityPowerMult})
  )
  n12 = MaterialFunctionCall(Texture Scale=Constant() {r=3}) {materialFunction=&ScaleUVsByCenter}
  n13 = Subtract(
   A=@n11
   B=Multiply(
    A=TextureSample(UVs=@n12.UVs) {samplerType=SAMPLERTYPE_Masks, texture=&T-FlowMap-vortex}.B
    B=ScalarParameter() {defaultValue=1.5, parameterName=Subtract Multi}
   ) {constB=1.25}
  ) {constB=0.25}
  n14 = ScalarParameter() {defaultValue=1.35, parameterName=ShapeFallof}
  Clamp(None=Power(Base=Multiply(A=@n13, B=@n14), Exp=ScalarParameter() {defaultValue=1.9006, parameterName=AlphaPower}))
  n15 = ScalarParameter() {defaultValue=4, parameterName=ColPower}
  LinearInterpolate(
   A=VectorParameter() {parameterName=Col01, defaultValue=<rgba 0 0.416 1 1>}.RGB
   B=VectorParameter() {parameterName=Cal02, defaultValue=<rgba 0.4427 1 1 1>}.RGB
   Alpha=Clamp(None=Power(Base=@n11, Exp=@n15))
  )
 ]
}
