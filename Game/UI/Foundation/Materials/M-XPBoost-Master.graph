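# Translucent UI master material that draws an XP boost icon as three vertically offset arrows.
# One half-arrow icon texture is sampled three times, with parameters choosing each arrow's color between two RGBA colors and weighting its mask.
# A rotated, time-panned shine texture blends in shine colors, using the rest-XP color for Arrow3 and the boost-XP color for Arrow1 and Arrow2.
# The weighted masks composite the result over a constant pink.
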Material /Game/UI/Foundation/Materials/M-XPBoost-Master
graph {
 properties {
  bUsedWithStaticMesh=false
  blendMode=BLEND_Translucent
  materialDomain=MD_UI
  pixelDepthOffsetMode=PDOM_Legacy
  shadingModel=MSM_Unlit
  usedShadingModels=Unlit
 }
 nodes [
  n1 = VectorParameter() {parameterName=RGBA0, defaultValue=<rgba 0 0.3005 1 1>}
  n2 = VectorParameter() {parameterName=RGBA1, defaultValue=<rgba 0.107 1 0.3515 1>}
  n3 = TextureCoordinate() {coordinateIndex=4, uTiling=2, vTiling=2.5}
  n4 = TextureSample(UVs=Add(A=@n3, B=Constant2Vector() {g=-0.0665})) {texture=/Game/UI/Foundation/Textures/Icons/Boost/Icon-Boost-OneHalf-S}
  n5 = Multiply(A=LinearInterpolate(A=@n1.RGB, B=@n2.RGB, Alpha=ScalarParameter() {desc="0=RGBA1 1=RGBA2", parameterName=Arrow3 Color}), B=@n4.A)
  n6 = ScalarParameter() {desc="0=RGBA1 1=RGBA2", parameterName=Arrow2 Color}
  n7 = TextureSample(UVs=Add(A=@n3, B=Constant2Vector() {g=-0.8})) {texture=&Icon-Boost-OneHalf-S}
  n8 = LinearInterpolate(B=Multiply(A=LinearInterpolate(A=@n1.RGB, B=@n2.RGB, Alpha=@n6), B=@n7.A), Alpha=@n7.A)
  n9 = TextureSample(UVs=Add(A=@n3, B=Constant2Vector() {g=-1.5508})) {texture=&Icon-Boost-OneHalf-S}
  n10 = Multiply(A=LinearInterpolate(A=@n1.RGB, B=@n2.RGB, Alpha=ScalarParameter() {desc="0=RGBA1 1=RGBA2", parameterName=Arrow1 Color}), B=@n9.A)
  n11 = Clamp(None=Add(A=Add(A=@n5, B=@n8), B=LinearInterpolate(B=@n10, Alpha=@n9.A)))
  n12 = VectorParameter() {parameterName=RestXPColor_Shine, defaultValue=<rgba 0.6038 1 0.9647 0>}
  n13 = VectorParameter() {parameterName=BoostXPColor_Shine, defaultValue=<rgba 0.6038 1 0.9647 0>}
  n14 = Add(
   A=Add(A=LinearInterpolate(B=@n12.RGB, Alpha=@n4.A), B=LinearInterpolate(B=@n13.RGB, Alpha=@n7.A))
   B=LinearInterpolate(B=@n13.RGB, Alpha=@n9.A)
  )
  n15 = Rotator(Coordinate=TextureCoordinate() {coordinateIndex=4, uTiling=0.4, vTiling=0.2}, Time=Constant() {r=1.9})
  n16 = Multiply(A=Time(), B=ScalarParameter() {defaultValue=1, parameterName=Shine Time})
  n17 = TextureSample(UVs=Frac(None=Panner(Coordinate=@n15, Time=@n16) {constCoordinate=4, speedX=-2, speedY=-1})) {texture=/Game/UI/Foundation/Textures/Pattern/T-UI-Shine}
  n18 = ScalarParameter() {defaultValue=0.5, parameterName=Shine Alpha, sliderMax=1}
  n19 = Multiply(A=@n4.A, B=ScalarParameter() {defaultValue=1, parameterName=Arrow3, sliderMax=1})
  n20 = Multiply(A=@n7.A, B=ScalarParameter() {defaultValue=1, parameterName=Arrow2, sliderMax=1})
  n21 = ScalarParameter() {defaultValue=1, parameterName=Arrow1, sliderMax=1}
  LinearInterpolate(
   A=Constant3Vector() {constant=<rgba 0.76 0.342 0.5831 0>}
   B=LinearInterpolate(A=@n11, B=Clamp(None=@n14), Alpha=Multiply(A=@n17.RGB, B=@n18))
   Alpha=Add(A=Add(A=@n19, B=@n20), B=Multiply(A=@n9.A, B=@n21))
  )
  Multiply(A=@n11) {constB=5}
 ]
}
