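# Material function for UI that fits a texture into a UI rectangle with letterboxing.
# A Custom HLSL node reads the texture's dimensions, which divide the UI Size input into per-axis ratios.
# Those ratios build two centered UV remaps, with a Step selecting between them without branching.
# The texture is sampled there, and ClampAlphaToTransparentBorder supplies the mask that blends it with Letter Box Color.
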MaterialFunction /Game/UI/Foundation/Materials/MF_UI_LetterBoxing
graph {nodes [
  MaterialExpressionFunctionInput_0 = FunctionInput(Preview=TextureObject()) {inputName=Draw Texture, inputType=FunctionInput_Texture2D}
  n1 = Custom!(Tex=@MaterialExpressionFunctionInput_0) {
   code="float2 dim;\r\nTextureObject.GetDimensions(dim.x, dim.y);\r\nreturn dim;"
   description=GetDimenstions
   outputType=CMOT_Float2
   inputs [{inputName=TextureObject, input {expression=&self:MaterialExpressionFunctionInput_0}}]
  }
  n2 = MaterialFunctionCall(Float2=Divide(A=FunctionInput() {inputName=UI Size, inputType=FunctionInput_Vector2}, B=@n1) {constB=222222221312}) {materialFunction=&BreakOutFloat2Components}
  n3 = Divide(A=Divide(A=@n2.G, B=@n2.G))
  n4 = Constant() {r=0.5}
  n5 = AppendVector(A=Constant(), B=Add(A=Multiply(A=@n3) {constB=-1}, B=@n4) {constB=0.5})
  n6 = AppendVector(A=Constant() {r=1}, B=Add(A=@n3, B=@n4) {constB=0.5})
  n7 = FunctionInput() {inputName=UI UVs, inputType=FunctionInput_Vector2}
  n8 = Step(Y=@n2.G, X=@n2.G) {constX=0, constY=1}
  n9 = Divide(A=Divide(A=@n2.R, B=@n2.R))
  n10 = Constant() {r=0.5}
  n11 = AppendVector(A=Add(A=Multiply(A=@n9) {constB=-1}, B=@n10) {constB=0.5}, B=Constant())
  n12 = AppendVector(A=Add(A=@n9, B=@n10) {constB=0.5}, B=Constant() {r=1})
  n13 = Add(
   A=Multiply(A=LinearInterpolate(A=@n5, B=@n6, Alpha=@n7), B=OneMinus(None=@n8))
   B=Multiply(A=LinearInterpolate(A=@n11, B=@n12, Alpha=@n7), B=@n8)
  )
  n14 = TextureSample(UVs=@n13, Tex=@MaterialExpressionFunctionInput_0)
  n15 = MaterialFunctionCall!(UVs=@n13, Alpha=@n14.A) {materialFunction=/Game/UI/Foundation/Items/ClampAlphaToTransparentBorder}
  n16 = OneMinus(None=@n15.Result)
  n17 = FunctionInput() {inputName=Letter Box Color}
  FunctionOutput(None=Add(A=Multiply(A=@n16, B=@n17), B=Multiply(A=@n15.Result, B=@n14.RGB)))
 ]}
