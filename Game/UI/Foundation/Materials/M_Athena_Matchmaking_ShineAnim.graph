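# Translucent UI material that draws an animated diagonal shine band, named for Athena matchmaking, optionally over an icon texture.
# ShineAnim slides the UVs between range parameters, then rotation, absolute value and SmoothStep form a band with adjustable thickness and softness.
# With UseIconRGBA on, the band color-dodges onto the icon, tintable toward an override color, using icon alpha.
# With it off, ShineRGB supplies color and the band supplies opacity.
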Material /Game/UI/Foundation/Materials/M_Athena_Matchmaking_ShineAnim
graph {
 properties {
  bUsedWithStaticMesh=false
  blendMode=BLEND_Translucent
  materialDomain=MD_UI
  pixelDepthOffsetMode=PDOM_Legacy
  usedShadingModels=Default Lit
 }
 nodes [
  Fmod(A=Time(), B=Constant() {r=1})
  n1 = ScalarParameter() {defaultValue=1, parameterName=ShineRGB}
  n2 = TextureCoordinate()
  n3 = AppendVector(
   A=LinearInterpolate(
    A=ScalarParameter() {defaultValue=1, parameterName=RangeStart}
    B=ScalarParameter() {defaultValue=-2, parameterName=RangeEnd}
    Alpha=ScalarParameter() {parameterName=ShineAnim}
   )
   B=Constant()
  )
  n4 = Abs(None=Rotator(Coordinate=Add(A=@n2, B=@n3) {constB=-0.5}, Time=ScalarParameter() {defaultValue=-0.125, parameterName=Rotation}) {centerX=0, centerY=0, speed=6.2832})
  n5 = ScalarParameter() {defaultValue=0.1, parameterName=Thickness}
  n6 = ScalarParameter() {parameterName=SSThreshold}
  n7 = Multiply(A=@n6) {constB=-1}
  n8 = OneMinus(None=MaterialFunctionCall(Alpha=ComponentMask(None=@n4) {r=true}, Min=Add(A=@n5, B=@n7), Max=Add(A=@n5, B=@n6)) {materialFunction=&SmoothStep}.Result)
  n9 = TextureSampleParameter2D() {parameterName=IconRGBA, texture=/Game/UI/Foundation/Textures/Icons/Items/T-FNBR-BattlePass-L}
  n10 = VectorParameter() {parameterName=RGBOverride, defaultValue=<rgba 1 1 0.0723 1>}
  n11 = ScalarParameter() {parameterName=RGBOverrideAnim}
  StaticSwitchParameter(
   True=MaterialFunctionCall!(Blend=Multiply(A=@n1, B=@n8) {constA=0.875}, Base=LinearInterpolate(A=@n9.RGB, B=@n10.RGB, Alpha=@n11)) {materialFunction=/Engine/Functions/Engine_MaterialFunctions03/Blends/Blend_ColorDodge}.Result
   False=@n1
  ) {parameterName=UseIconRGBA}
  StaticSwitchParameter(True=@n9.A, False=@n8) {parameterName=UseIconRGBA}
 ]
}
