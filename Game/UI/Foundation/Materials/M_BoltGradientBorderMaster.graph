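# Translucent UI master material that draws a two-color gradient, with alpha, cut by procedural bolt-shaped borders.
# The gradient uses UV channel 4 with adjustable position, scale and rotation.
# Rotated UVs form angled left and right edges, joined along a slanted center seam, with sharpness multipliers controlling softness.
# UseTopBolt blends in a second such edge, intersecting both masks through a scalar DotProduct, before OverallAlpha scales the opacity.
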Material /Game/UI/Foundation/Materials/M_BoltGradientBorderMaster
graph {
 properties {
  bUsedWithStaticMesh=false
  blendMode=BLEND_Translucent
  materialDomain=MD_UI
  pixelDepthOffsetMode=PDOM_Legacy
  usedShadingModels=Default Lit
 }
 nodes [
  n1 = VectorParameter() {parameterName=RGBA_Left, defaultValue=<rgba 0 0.4219 0.0328 0.5>}
  n2 = VectorParameter() {parameterName=RGBA_Right, defaultValue=<rgba 0 1 0.0778 1>}
  n3 = ScalarParameter() {group=ColorGradient, parameterName=GradientPosition_Y}
  n4 = ScalarParameter() {group=ColorGradient, parameterName=GradientPosition_X}
  n5 = Divide(
   A=Add(A=TextureCoordinate() {coordinateIndex=4}, B=AppendVector(A=@n3, B=@n4))
   B=ScalarParameter() {defaultValue=1, group=ColorGradient, parameterName=GradientScale}
  )
  n6 = Divide(A=ScalarParameter() {group=ColorGradient, parameterName=GradientRotation}) {constB=360}
  n7 = Clamp(None=ComponentMask(None=Rotator(Coordinate=@n5, Time=@n6) {speed=6.2832}) {r=true})
  LinearInterpolate(A=@n1.RGB, B=@n2.RGB, Alpha=@n7)
  n8 = ScalarParameter() {defaultValue=10, group="Angles(Degrees)", parameterName=LeftAngle}
  n9 = TextureCoordinate()
  n10 = Add(
   A=Rotator(Coordinate=@n9, Time=Divide(A=@n8) {constB=360}) {centerY=0, speed=6.2832}
   B=Add(A=ScalarParameter() {defaultValue=0.175, group=Offsets, parameterName=LeftOffset}) {constB=-1}
  )
  n11 = ScalarParameter() {defaultValue=32, parameterName=LeftRightEdgeSharpness}
  n12 = ScalarParameter() {defaultValue=10, group="Angles(Degrees)", parameterName=RightAngle}
  n13 = Add(
   A=Rotator(Coordinate=@n9, Time=Divide(A=@n12) {constB=360}) {centerY=1, speed=6.2832}
   B=Add(A=ScalarParameter() {group=Offsets, parameterName=RightOffset}) {constB=-1}
  )
  n14 = Add(A=@n9) {constB=-0.5}
  n15 = Divide(A=ScalarParameter() {defaultValue=-20, group="Angles(Degrees)", parameterName=CenterAngle}) {constB=360}
  n16 = ScalarParameter() {group=Offsets, parameterName=CenterOffset}
  n17 = ComponentMask(None=Add(A=Rotator(Coordinate=@n14, Time=@n15) {centerX=1, speed=6.2832}, B=Multiply(A=@n16) {constB=-1})) {r=true}
  n18 = ScalarParameter() {defaultValue=32, parameterName=CenterEdgeSharpness}
  n19 = LinearInterpolate(
   A=Multiply(A=ComponentMask(None=@n10) {g=true}, B=@n11)
   B=Multiply(A=ComponentMask(None=@n13) {g=true}, B=@n11)
   Alpha=Clamp(None=Multiply(A=@n17, B=@n18))
  )
  n20 = Rotator(Coordinate=@n9, Time=Divide(A=ScalarParameter() {defaultValue=10, group="Angles(Degrees)", parameterName=TopLeftAngle}) {constB=360}) {centerY=0, speed=6.2832}
  n21 = Add(A=ScalarParameter() {defaultValue=0.175, group=Offsets, parameterName=TopLeftOffset}) {constB=-0.125}
  n22 = ScalarParameter() {defaultValue=32, parameterName=TopLeftRightEdgeSharpness}
  n23 = Multiply(A=ComponentMask(None=Add(A=@n20, B=@n21)) {g=true}, B=@n22)
  n24 = Rotator(Coordinate=@n9, Time=Divide(A=ScalarParameter() {defaultValue=10, group="Angles(Degrees)", parameterName=TopRightAngle}) {constB=360}) {centerY=1, speed=6.2832}
  n25 = Add(A=ScalarParameter() {group=Offsets, parameterName=TopRightOffset}) {constB=-0.125}
  n26 = Multiply(A=ComponentMask(None=Add(A=@n24, B=@n25)) {g=true}, B=@n22)
  n27 = TextureCoordinate()
  n28 = ScalarParameter() {defaultValue=-20, group="Angles(Degrees)", parameterName=TopCenterAngle}
  n29 = Add(
   A=Rotator(Coordinate=Add(A=@n27) {constB=-0.5}, Time=Divide(A=@n28) {constB=360}) {centerX=1, speed=6.2832}
   B=Multiply(A=ScalarParameter() {group=Offsets, parameterName=TopCenterOffset}) {constB=-1}
  )
  n30 = Clamp(None=Multiply(A=ComponentMask(None=@n29) {r=true}, B=ScalarParameter() {defaultValue=32, parameterName=TopCenterEdgeSharpness}))
  n31 = DotProduct(A=OneMinus(None=Clamp(None=@n19)), B=Clamp(None=LinearInterpolate(A=@n23, B=@n26, Alpha=@n30)))
  n32 = Clamp(None=LinearInterpolate(A=@n19, B=OneMinus(None=@n31), Alpha=ScalarParameter() {parameterName=UseTopBolt}))
  Multiply(
   A=LinearInterpolate(B=LinearInterpolate(A=@n1.A, B=@n2.A, Alpha=@n7), Alpha=OneMinus(None=@n32))
   B=ScalarParameter() {defaultValue=1, parameterName=OverallAlpha}
  )
 ]
}
