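# Translucent UI material that draws three layers of scrolling triangles for a Battle Pass effect, fading along a selectable axis.
# Three panners at different speeds feed customized UVs, which sample separate channels of one triangle texture.
# SmoothStep thresholds shift along a UV gradient to vary each layer's triangle cutoff.
# Color blends between a gradient and per-layer tints, while opacity combines weighted layers with edge fades.
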Material /Game/UI/Foundation/Materials/M_FX_TriangleScroll_BattlePass
graph {
 properties {
  bUsedWithStaticMesh=false
  blendMode=BLEND_Translucent
  materialDomain=MD_UI
  numCustomizedUVs=4
  pixelDepthOffsetMode=PDOM_Legacy
  usedShadingModels=Default Lit
 }
 nodes [
  n1 = VectorParameter() {parameterName=GradRGB1, defaultValue=<rgba 1 1 0.3185 1>}
  n2 = VectorParameter() {parameterName=GradRGB2, defaultValue=<rgba 1 1 0.0723 1>}
  n3 = StaticComponentMaskParameter(None=TextureCoordinate()) {defaultR=true, parameterName=Direction}
  n4 = TextureSampleParameter2D(UVs=TextureCoordinate() {coordinateIndex=1}) {
   constCoordinate=1
   parameterName=TriangleTexture
   samplerType=SAMPLERTYPE_LinearColor
   texture=/Game/UI/Foundation/Textures/Pattern/T_Triangle_BattlePass
  }
  n5 = Constant()
  n6 = TextureCoordinate()
  n7 = Power(
   Base=StaticComponentMaskParameter(None=@n6) {defaultR=true, parameterName=Direction}
   Exp=ScalarParameter() {defaultValue=1, parameterName=FallOffPower}
  ) {constExponent=4}
  n8 = Multiply(A=@n6) {constB=8}
  n9 = Clamp(None=OneMinus(None=StaticComponentMaskParameter(None=@n8) {defaultR=true, parameterName=Direction}))
  n10 = Clamp(None=Add(A=Clamp(None=@n7), B=@n9)) {maxDefault=0.75}
  n11 = LinearInterpolate(
   A=ScalarParameter() {defaultValue=0.25, parameterName=ssMarginMax}
   B=ScalarParameter() {defaultValue=0.0625, parameterName=ssMarginMin}
   Alpha=@n10
  )
  n12 = Subtract(A=Constant() {r=1}, B=@n11)
  n13 = LinearInterpolate(A=@n5, B=@n12, Alpha=@n10)
  n14 = LinearInterpolate(A=Add(A=@n5, B=@n11), B=Add(A=@n12, B=@n11), Alpha=@n10)
  n15 = MaterialFunctionCall(Alpha=@n4.G, Min=@n13, Max=@n14) {materialFunction=&SmoothStep}
  n16 = LinearInterpolate(A=VectorParameter() {parameterName=AlphaEdges, defaultValue=<rgba 0 0 0 1>}.RGB, B=@n1.RGB, Alpha=@n15.Result)
  n17 = TextureSampleParameter2D(UVs=TextureCoordinate() {coordinateIndex=2}) {parameterName=TriangleTexture, samplerType=SAMPLERTYPE_LinearColor, texture=&T_Triangle_BattlePass}
  n18 = MaterialFunctionCall(Alpha=@n17.B, Min=@n13, Max=@n14) {materialFunction=&SmoothStep}
  n19 = TextureSampleParameter2D(UVs=TextureCoordinate() {coordinateIndex=3}) {parameterName=TriangleTexture, samplerType=SAMPLERTYPE_LinearColor, texture=&T_Triangle_BattlePass}
  n20 = MaterialFunctionCall(Alpha=@n19.R, Min=@n13, Max=@n14) {materialFunction=&SmoothStep}
  LinearInterpolate(
   A=LinearInterpolate(A=@n1.RGB, B=@n2.RGB, Alpha=Power(Base=@n3) {constExponent=0.35})
   B=LinearInterpolate(
    A=LinearInterpolate(A=@n16, B=@n2.RGB, Alpha=@n18.Result)
    B=VectorParameter() {parameterName=GradRGB3, defaultValue=<rgba 1 1 0.0723 1>}.RGB
    Alpha=@n20.Result
   )
   Alpha=ScalarParameter() {parameterName=ColorPerRGB}
  )
  n21 = OneMinus(None=@n9)
  n22 = TextureCoordinate() {uTiling=0.5}
  n23 = TextureCoordinate() {vTiling=0.5}
  n24 = ScalarParameter() {parameterName=SpeedDirection}
  n25 = StaticComponentMaskParameter(None=Add(A=LinearInterpolate(A=@n22, B=@n23, Alpha=@n24), B=Constant2Vector())) {defaultR=true, parameterName=Direction}
  n26 = Clamp(None=OneMinus(None=Multiply(A=@n25) {constB=2}))
  n27 = Multiply(A=@n15.Result)
  n28 = Multiply(A=@n18.Result, B=ScalarParameter() {defaultValue=0.66, parameterName=Layer2Opacity}) {constB=0.66}
  n29 = Multiply(A=@n20.Result, B=ScalarParameter() {defaultValue=0.33, parameterName=Layer3Opacity}) {constB=0.33}
  n30 = Add(A=@n27, B=Add(A=@n28, B=@n29))
  n31 = LinearInterpolate(A=@n27, Alpha=LinearInterpolate(A=@n28, Alpha=@n29))
  n32 = ScalarParameter() {parameterName=LerpMaskLayers}
  Multiply(
   A=Multiply(A=Multiply(A=@n21, B=@n26), B=LinearInterpolate(A=@n30, B=@n31, Alpha=@n32))
   B=ScalarParameter() {defaultValue=0.6, parameterName=OpacityOverride}
  )
  n33 = Multiply(
   A=TextureCoordinate()
   B=AppendVector(A=ScalarParameter() {defaultValue=3, parameterName=TileU}, B=ScalarParameter() {defaultValue=1, parameterName=TileV})
  ) {constB=3}
  n34 = ScalarParameter() {defaultValue=-0.5, parameterName=TriangleUSpeedA}
  n35 = Constant()
  n36 = ScalarParameter() {parameterName=SpeedDirection}
  Panner(Coordinate=@n33, Speed=LinearInterpolate(A=AppendVector(A=@n34, B=@n35), B=AppendVector(A=@n35, B=@n34), Alpha=@n36)) {speedX=-0.09}
  n37 = ScalarParameter() {defaultValue=-0.4, parameterName=TriangleUSpeedB}
  Panner(Coordinate=@n33, Speed=LinearInterpolate(A=AppendVector(A=@n37, B=@n35), B=AppendVector(A=@n35, B=@n37), Alpha=@n36)) {speedX=-0.07}
  n38 = ScalarParameter() {defaultValue=-0.3, parameterName=TriangleUSpeedC}
  Panner(Coordinate=@n33, Speed=LinearInterpolate(A=AppendVector(A=@n38, B=@n35), B=AppendVector(A=@n35, B=@n38), Alpha=@n36)) {speedX=-0.05}
 ]
}
