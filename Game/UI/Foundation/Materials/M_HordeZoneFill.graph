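# Translucent UI material that, going by its name, fills a horde zone with a flat color shaped as a quad with slanted edges.
# Each edge comes from UVs rotated about a corner by an inclination parameter, with static switches choosing the pivot corner.
# Left/right edges use UV0 and top/bottom use UV4, each pair multiplied and smoothstepped.
# The combined, clamped mask is scaled by AlphaOverride, alongside an RGB color parameter.
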Material /Game/UI/Foundation/Materials/M_HordeZoneFill
graph {
 properties {
  bUsedWithStaticMesh=false
  blendMode=BLEND_Translucent
  materialDomain=MD_UI
  pixelDepthOffsetMode=PDOM_Legacy
  usedShadingModels=Default Lit
 }
 nodes [
  n1 = ScalarParameter() {defaultValue=-6.5, parameterName=Inclination Left Top}
  n2 = TextureCoordinate()
  n3 = ScalarParameter() {defaultValue=6.5, parameterName=Inclination Left Bottom}
  n4 = StaticSwitch(
   True=Rotator(Coordinate=@n2, Time=Divide(A=@n1) {constA=360, constB=360}) {centerX=0, centerY=1, speed=1}
   False=Rotator(Coordinate=@n2, Time=Divide(A=@n3) {constA=360, constB=360}) {centerX=0, centerY=0, speed=1}
   Value=StaticBoolParameter() {group=Direction Switches, parameterName=Direction left side, sortPriority=2}
  )
  n5 = Divide(A=ScalarParameter() {defaultValue=-2.5, parameterName=Inclination Right Bottom}) {constA=360, constB=360}
  n6 = Divide(A=ScalarParameter() {defaultValue=-2.5, parameterName=Inclination Right Top}) {constA=360, constB=360}
  n7 = OneMinus(None=StaticSwitch(
    True=Rotator(Coordinate=@n2, Time=@n5) {centerX=1, centerY=0, speed=1}
    False=Rotator(Coordinate=@n2, Time=@n6) {centerX=1, centerY=1, speed=1}
    Value=StaticBoolParameter() {group=Direction Switches, parameterName=Direction Right side, sortPriority=3}
   ))
  n8 = MaterialFunctionCall(
   Alpha=DotProduct(A=ComponentMask(None=@n4) {r=true}, B=ComponentMask(None=@n7) {r=true})
   Max=ScalarParameter() {defaultValue=0.01, parameterName=LeftRightSSMax}
  ) {materialFunction=&SmoothStep}
  n9 = ScalarParameter() {defaultValue=2.5, parameterName=Inclination Top Left}
  n10 = TextureCoordinate() {coordinateIndex=4}
  n11 = ScalarParameter() {defaultValue=-2.5, parameterName=Inclination Top Right}
  n12 = StaticSwitch(
   True=Rotator(Coordinate=@n10, Time=Divide(A=@n9) {constA=360, constB=360}) {centerX=1, centerY=0, speed=1}
   False=Rotator(Coordinate=@n10, Time=Divide(A=@n11) {constA=360, constB=360}) {centerX=0, centerY=0, speed=1}
   Value=StaticBoolParameter() {group=Direction Switches, parameterName=Direction Top}
  )
  n13 = Divide(A=ScalarParameter() {defaultValue=-2, parameterName=Inclination Bottom Right}) {constA=360, constB=360}
  n14 = Divide(A=ScalarParameter() {defaultValue=-2, parameterName=Inclination Bottom Left}) {constA=360, constB=360}
  n15 = OneMinus(None=StaticSwitch(
    True=Rotator(Coordinate=@n10, Time=@n13) {centerX=0, centerY=1, speed=1}
    False=Rotator(Coordinate=@n10, Time=@n14) {centerX=1, centerY=1, speed=1}
    Value=StaticBoolParameter() {group=Direction Switches, parameterName=Direction Bottom, sortPriority=1}
   ))
  n16 = MaterialFunctionCall(
   Alpha=DotProduct(A=ComponentMask(None=@n12) {g=true}, B=ComponentMask(None=@n15) {g=true})
   Max=ScalarParameter() {defaultValue=0.02, parameterName=TopBottomSSMax}
  ) {materialFunction=&SmoothStep}
  Multiply(A=ScalarParameter() {defaultValue=0.7, parameterName=AlphaOverride}, B=Clamp(None=DotProduct(A=@n8.Result, B=@n16.Result)))
  VectorParameter() {parameterName=RGB, defaultValue=<rgba 1 0 0.5055 1>}
 ]
}
