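# Translucent UI material, named for the Horde zone label, that draws a slanted-edge label shape procedurally, without textures.
# Rotated UV gradients through SmoothStep form the outer silhouette, with IsArrowStyle blending to a variant pointed by two angled cuts.
# Shifted inner edges separate outline from fill, with an optional separate fill colour via SplitFillAndOutlineRGB.
# Opacity combines outline plus InnerFillOpacity with the silhouette, scaled by AlphaOverride.
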Material /Game/UI/Foundation/Materials/M_HordeZoneLabel
graph {
 properties {
  bUsedWithStaticMesh=false
  blendMode=BLEND_Translucent
  materialDomain=MD_UI
  pixelDepthOffsetMode=PDOM_Legacy
  usedShadingModels=Default Lit
 }
 nodes [
  n1 = VectorParameter() {parameterName=RGB, defaultValue=<rgba 1 0 0.5055 1>}
  n2 = TextureCoordinate()
  n3 = ComponentMask(None=Add(A=@n2, B=ScalarParameter() {defaultValue=-0.14, parameterName=TopInnerEdgeShift}) {constB=-0.14}) {g=true}
  n4 = ComponentMask(None=Add(A=@n2, B=ScalarParameter() {defaultValue=-0.88, parameterName=BottomInnerEdgeShift}) {constB=-0.88}) {g=true}
  n5 = ScalarParameter() {defaultValue=0.02, parameterName=TopBottomSSMax}
  n6 = Add(A=@n2, B=ScalarParameter() {defaultValue=-0.03, parameterName=LeftInnerEdgeShift}) {constB=-0.03}
  n7 = Divide(A=Constant() {r=-7.5}) {constA=360, constB=360}
  n8 = Rotator(
   Coordinate=Add(A=@n2, B=ScalarParameter() {defaultValue=0.045, parameterName=RightInnerEdgeShift}) {constB=0.045}
   Time=Divide(A=Constant() {r=-5.5}) {constA=360, constB=360}
  ) {centerX=1, centerY=0, speed=6.2832}
  n9 = DotProduct(
   A=ComponentMask(None=Rotator(Coordinate=@n6, Time=@n7) {centerX=0, centerY=1, speed=6.2832}) {r=true}
   B=ComponentMask(None=OneMinus(None=@n8)) {r=true}
  )
  n10 = ScalarParameter() {defaultValue=0.01, parameterName=LeftRightSSMax}
  n11 = LinearInterpolate(
   A=MaterialFunctionCall(Alpha=DotProduct(A=@n3, B=@n4), Max=@n5) {materialFunction=&SmoothStep}.Result
   Alpha=OneMinus(None=MaterialFunctionCall(Alpha=@n9, Max=@n10) {materialFunction=&SmoothStep}.Result)
  )
  LinearInterpolate(
   A=@n1.RGB
   B=LinearInterpolate(A=VectorParameter() {parameterName=InnerFillRGB, defaultValue=<rgba 1 0.966 0 1>}.RGB, B=@n1.RGB, Alpha=@n11)
   Alpha=ScalarParameter() {parameterName=SplitFillAndOutlineRGB}
  )
  n12 = Add(A=@n11, B=ScalarParameter() {parameterName=InnerFillOpacity})
  n13 = Constant() {r=-6.5}
  n14 = ComponentMask(None=Rotator(Coordinate=@n2, Time=Divide(A=@n13) {constA=360, constB=360}) {centerX=0, centerY=1, speed=6.2832}) {r=true}
  n15 = Rotator(Coordinate=@n2, Time=Divide(A=Constant() {r=-2.5}) {constA=360, constB=360}) {centerX=1, centerY=0, speed=6.2832}
  n16 = DotProduct(A=@n14, B=ComponentMask(None=OneMinus(None=@n15)) {r=true})
  n17 = TextureCoordinate() {coordinateIndex=4}
  n18 = Rotator(Coordinate=@n17, Time=Divide(A=Constant() {r=2.5}) {constA=360, constB=360}) {centerX=1, centerY=0, speed=6.2832}
  n19 = Constant() {r=2}
  n20 = OneMinus(None=Rotator(Coordinate=@n17, Time=Divide(A=@n19) {constA=360, constB=360}) {centerX=0, centerY=1, speed=6.2832})
  n21 = MaterialFunctionCall(Alpha=DotProduct(A=ComponentMask(None=@n18) {g=true}, B=ComponentMask(None=@n20) {g=true}), Max=@n5) {materialFunction=&SmoothStep}
  n22 = MaterialFunctionCall(Alpha=@n14, Max=@n10) {materialFunction=&SmoothStep}
  n23 = Constant() {r=25}
  n24 = OneMinus(None=Rotator(Coordinate=@n2, Time=Divide(A=@n23) {constA=360, constB=360}) {centerX=1, speed=6.2832})
  n25 = Constant() {r=-30}
  n26 = OneMinus(None=Rotator(Coordinate=@n2, Time=Divide(A=@n25) {constA=360, constB=360}) {centerX=1, speed=6.2832})
  n27 = DotProduct(
   A=MaterialFunctionCall(Alpha=ComponentMask(None=@n24) {r=true}, Max=@n10) {materialFunction=&SmoothStep}.Result
   B=MaterialFunctionCall(Alpha=ComponentMask(None=@n26) {r=true}, Max=@n10) {materialFunction=&SmoothStep}.Result
  )
  n28 = LinearInterpolate(
   A=DotProduct(
    A=MaterialFunctionCall(Alpha=@n16, Max=@n10) {materialFunction=&SmoothStep}.Result
    B=@n21.Result
   )
   B=DotProduct(A=@n21.Result, B=Multiply(A=@n22.Result, B=@n27))
   Alpha=ScalarParameter() {parameterName=IsArrowStyle}
  )
  Multiply(A=ScalarParameter() {defaultValue=0.7, parameterName=AlphaOverride}, B=Clamp(None=Multiply(A=@n12, B=@n28)))
 ]
}
