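# Translucent UI master material for menu buttons, drawing an outline and gradient fill procedurally from UVs.
# A box-edge mask, optionally extended by an aspect-corrected angled cut, separates outline from interior, driving color blends and opacity.
# Offset parameters split fill visibility horizontally, and a scalar blends in a ShineSwipe function.
# Static switches add a spinning shine from two rotating DrawLine-2D streaks, masked to the outline, to color and opacity.
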Material /Game/UI/Foundation/Materials/M_MenuButton_Master
graph {
 properties {
  bUsedWithStaticMesh=false
  blendMode=BLEND_Translucent
  materialDomain=MD_UI
  numCustomizedUVs=2
  pixelDepthOffsetMode=PDOM_Legacy
  usedShadingModels=Default Lit
 }
 nodes [
  n1 = TextureCoordinate()
  n2 = Subtract(A=Abs(None=Add(A=@n1) {constB=-0.5}), B=ScalarParameter() {defaultValue=0.433, parameterName=Box Size})
  Add(A=ComponentMask(None=@n2) {r=true}, B=ComponentMask(None=@n2) {g=true})
  n3 = Clamp(None=Multiply(A=@n2) {constB=1000})
  Clamp(None=Add(A=ComponentMask(None=@n3) {r=true}, B=ComponentMask(None=@n3) {g=true}))
  n4 = VectorParameter() {parameterName=Primary Color, defaultValue=<rgba 0.3185 0.6038 1 1>}
  n5 = VectorParameter() {parameterName=Base Color, defaultValue=<rgba 0 0.0021 0.0176 1>}
  n6 = ComponentMask(None=MaterialFunctionCall() {materialFunction=&GetUserInterfaceUV}.9-Slice UV) {g=true}
  n7 = Constant()
  n8 = Constant() {r=1}
  n9 = Clamp(None=OneMinus(None=MaterialFunctionCall(Alpha=@n6, Min=@n7, Max=@n8) {materialFunction=&SmoothStep}.Result))
  n10 = Add(A=MaterialFunctionCall(Vector 2=Ceil(None=@n2)) {materialFunction=/Engine/Functions/Engine_MaterialFunctions02/Utility/VectorLength}.V2 Length) {constB=0}
  n11 = MaterialFunctionCall() {materialFunction=&GetUserInterfaceUV}
  n12 = MaterialFunctionCall(Float2=@n11.Pixel Size) {materialFunction=&BreakOutFloat2Components}
  n13 = Divide(A=@n12.R, B=Max(A=@n12.G) {constB=0.1})
  n14 = AppendVector(A=@n13, B=Constant() {r=1})
  n15 = Multiply(A=Subtract(A=@n13) {constB=1.125}) {constB=-1}
  n16 = Constant()
  n17 = Rotator(Coordinate=Add(A=Multiply(A=@n11.Normalized UV, B=@n14), B=AppendVector(A=@n15, B=@n16)), Time=Constant() {r=-0.0833}) {speed=6.2832}
  n18 = Multiply(A=MaterialFunctionCall(Alpha=ComponentMask(None=@n17) {r=true}, Min=Constant() {r=0.115}, Max=Constant() {r=0.125}) {materialFunction=&SmoothStep}.Result)
  n19 = Clamp(None=LinearInterpolate(A=@n10, B=LinearInterpolate(A=@n10, Alpha=@n18), Alpha=ScalarParameter() {parameterName=ToggleAngle}))
  n20 = OneMinus(None=@n19)
  n21 = Multiply(A=OneMinus(None=@n9), B=@n20)
  n22 = ScalarParameter() {defaultValue=0.6, parameterName=Dark Base Alpha, sliderMax=1}
  n23 = LinearInterpolate(
   A=LinearInterpolate(A=@n4.RGB, B=@n5.RGB, Alpha=Multiply(A=@n21, B=@n22))
   B=LinearInterpolate(A=@n4.RGB, B=@n5.RGB, Alpha=@n20)
   Alpha=ScalarParameter() {parameterName=SplitOutlineColor}
  )
  n24 = MaterialFunctionCall!(In_Uv) {materialFunction=&/M-Function-ShineSwipe}
  n25 = ScalarParameter() {parameterName=s_UseShineSwipe}
  n26 = LinearInterpolate(A=@n23, B=LinearInterpolate(A=@n23, B=@n24.Result, Alpha=@n24.Result) {constAlpha=0}, Alpha=@n25) {constAlpha=0}
  n27 = VectorParameter() {parameterName=SpinnyShineColor, defaultValue=<rgba 0.9165 1 0.9083 1>}
  n28 = ScalarParameter() {defaultValue=1, parameterName=SpinnyBrightness}
  n29 = ComponentMask(None=@n1) {r=true}
  n30 = ScalarParameter() {parameterName=Inner Fill Offset, sliderMax=1}
  n31 = Ceil(None=MaterialFunctionCall(Alpha=Subtract(A=@n29, B=@n30) {constB=0.5}, Min=Constant(), Max=Constant() {r=1}) {materialFunction=&SmoothStep}.Result)
  n32 = ScalarParameter() {parameterName=Switch Offset, sliderMax=1}
  n33 = Add(A=Multiply(A=OneMinus(None=@n31), B=@n32), B=Multiply(A=@n31, B=OneMinus(None=@n32)))
  n34 = Multiply(
   A=LinearInterpolate(A=@n9, Alpha=ScalarParameter() {defaultValue=0.75, parameterName=Gradient Min Alpha, sliderMax=1}) {constAlpha=0}
   B=ScalarParameter() {defaultValue=0.4, parameterName=Gradient Global Alpha, sliderMax=1}
  )
  n35 = Multiply(A=@n19, B=ScalarParameter() {defaultValue=0.8, parameterName=Outline Alpha, sliderMax=1})
  n36 = Multiply(A=Clamp(None=Add(A=@n33, B=@n19)), B=Add(A=Multiply(A=@n20, B=@n34), B=@n35))
  n37 = LinearInterpolate(A=@n36, B=LinearInterpolate(A=@n36, Alpha=@n24.Alpha), Alpha=@n25)
  n38 = MaterialFunctionCall() {materialFunction=&GetUserInterfaceUV}
  n39 = Rotator(Coordinate=@n38.Normalized UV) {speed=0.15}
  n40 = Constant2Vector() {g=0.5, r=-0.9}
  n41 = Constant2Vector() {g=0.5, r=1.1}
  n42 = Constant() {r=0.1}
  n43 = ScalarParameter() {defaultValue=2.1762, parameterName=EdgeFalloffSpinnyFlare}
  n44 = Rotator(Coordinate=@n38.Normalized UV) {speed=0.6}
  n45 = Constant() {r=0.08}
  n46 = Clamp(None=Add(
    A=MaterialFunctionCall!(Coordinates=@n39, P0=@n40, P1=@n41, Line Width=@n42, Edge Falloff=@n43, Square Corners?) {materialFunction=/Engine/Functions/Engine_MaterialFunctions01/Coordinates/DrawLine-2D}.Result
    B=MaterialFunctionCall(Coordinates=@n44, P0=@n40, P1=@n41, Line Width=@n45, Edge Falloff=@n43) {materialFunction=&DrawLine-2D}.Result
   ))
  n47 = Add(A=@n37, B=Multiply(A=Multiply(A=@n46, B=@n35), B=ScalarParameter() {defaultValue=1, parameterName=SpinnyOpacity}))
  StaticSwitchParameter(True=LinearInterpolate(A=@n26, B=Multiply(A=@n27.RGB, B=@n28), Alpha=@n47), False=@n26) {parameterName=UseSpinnyShineBorder}
  StaticSwitchParameter(True=@n47, False=@n37) {parameterName=UseSpinnyShineOpacity}
  TextureCoordinate() {coordinateIndex=1}
 ]
}
