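# Translucent UI material for series name backplates, building a procedural quadrilateral mask without textures.
# Centered UVs are rotated separately for X and Y, then mirrored with Abs to form two symmetric edge bands.
# Each band has parameters for rotation, edge inset and falloff, and a DotProduct of the clamped scalars intersects them.
# RGB and InnerFillRGB color parameters are declared, with no connections shown to the mask or outputs.
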Material /Game/UI/Foundation/Materials/M_SeriesNameBackplates
graph {
 properties {
  bUsedWithStaticMesh=false
  blendMode=BLEND_Translucent
  materialDomain=MD_UI
  pixelDepthOffsetMode=PDOM_Legacy
  usedShadingModels=Default Lit
 }
 nodes [
  n1 = Add(A=TextureCoordinate()) {constB=-0.5}
  n2 = Abs(None=Rotator(Coordinate=@n1, Time=ScalarParameter() {group=Edge-X, parameterName=Rotation X}) {centerX=0, centerY=0, speed=6.2832})
  n3 = Subtract(A=ComponentMask(None=OneMinus(None=@n2)) {r=true}, B=ScalarParameter() {group=Edge-X, parameterName=Edge-X})
  n4 = ScalarParameter() {defaultValue=64, group=Edge-X, parameterName=Edge-X-Falloff}
  n5 = Abs(None=Rotator(Coordinate=@n1, Time=ScalarParameter() {group=Edge-Y, parameterName=Rotation Y}) {centerX=0, centerY=0, speed=6.2832})
  n6 = Subtract(A=ComponentMask(None=OneMinus(None=@n5)) {g=true}, B=ScalarParameter() {group=Edge-Y, parameterName=Edge-Y})
  n7 = ScalarParameter() {defaultValue=64, group=Edge-Y, parameterName=Edge-Y-Falloff}
  DotProduct(A=Clamp(None=Multiply(A=@n3, B=@n4)), B=Clamp(None=Multiply(A=@n6, B=@n7)))
  VectorParameter() {parameterName=RGB, defaultValue=<rgba 1 0 0.5055 1>}
  VectorParameter() {parameterName=InnerFillRGB, defaultValue=<rgba 1 0.966 0 1>}
 ]
}
