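# M_SimpleSkewBox is a translucent UI material that draws a colored box with independently skewed edges.
# Each edge rotates UVs by an angle parameter and applies an offset, with top/bottom reading UV channel 4 and left/right channel 0.
# Opposing edges multiply into sharpness-scaled, clamped masks, whose product scales the RGBA alpha.
# A static switch, bIsGradient, swaps the flat color for a vertical gradient between RGBA1 and RGBA2.
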Material /Game/UI/Foundation/Materials/M_SimpleSkewBox
graph {
 properties {
  bUsedWithStaticMesh=false
  blendMode=BLEND_Translucent
  materialDomain=MD_UI
  pixelDepthOffsetMode=PDOM_Legacy
  usedShadingModels=Default Lit
 }
 nodes [
  n1 = VectorParameter() {parameterName=RGBA, defaultValue=<rgba 0 0.4219 0.0328 0.5>}
  n2 = Divide(A=ScalarParameter() {defaultValue=10, group="Angles(Degrees)", parameterName=TopAngle}) {constB=360}
  n3 = TextureCoordinate() {coordinateIndex=4}
  n4 = ScalarParameter() {defaultValue=0.125, group=Offsets, parameterName=TopOffset}
  n5 = ComponentMask(None=Add(A=Rotator(Coordinate=@n3, Time=@n2) {centerY=0, speed=6.2832}, B=Multiply(A=@n4) {constB=-1})) {g=true}
  n6 = ScalarParameter() {defaultValue=8, parameterName=TopBottomEdgeSharpness}
  n7 = Divide(A=ScalarParameter() {defaultValue=10, group="Angles(Degrees)", parameterName=BottomAngle}) {constB=360}
  n8 = ScalarParameter() {defaultValue=0.125, group=Offsets, parameterName=BottomOffset}
  n9 = ComponentMask(None=Add(A=Rotator(Coordinate=@n3, Time=@n7) {centerY=1, speed=6.2832}, B=Add(A=@n8) {constB=-1})) {g=true}
  n10 = OneMinus(None=DotProduct(A=Multiply(A=@n5, B=@n6), B=Multiply(A=@n9, B=@n6)))
  n11 = Divide(A=ScalarParameter() {defaultValue=10, group="Angles(Degrees)", parameterName=LeftAngle}) {constB=360}
  n12 = TextureCoordinate()
  n13 = ScalarParameter() {defaultValue=0.125, group=Offsets, parameterName=LeftOffset}
  n14 = ComponentMask(None=Add(A=Rotator(Coordinate=@n12, Time=@n11) {centerX=1, speed=6.2832}, B=Multiply(A=@n13) {constB=-1})) {r=true}
  n15 = ScalarParameter() {defaultValue=8, parameterName=RightLeftEdgeSharpness}
  n16 = Divide(A=ScalarParameter() {defaultValue=10, group="Angles(Degrees)", parameterName=RightAngle}) {constB=360}
  n17 = ScalarParameter() {defaultValue=0.125, group=Offsets, parameterName=RightOffset}
  n18 = ComponentMask(None=Add(A=Rotator(Coordinate=@n12, Time=@n16) {centerX=0, speed=6.2832}, B=Add(A=@n17) {constB=-1})) {r=true}
  n19 = OneMinus(None=DotProduct(A=Multiply(A=@n14, B=@n15), B=Multiply(A=@n18, B=@n15)))
  LinearInterpolate(B=@n1.A, Alpha=DotProduct(A=Clamp(None=@n10), B=Clamp(None=@n19)))
  n20 = TextureCoordinate()
  StaticSwitchParameter(
   True=LinearInterpolate(
    A=VectorParameter() {parameterName=RGBA1, defaultValue=<rgba 0 0.4219 0.0328 0.5>}.RGB
    B=VectorParameter() {parameterName=RGBA2, defaultValue=<rgba 0 0.4219 0.0328 0.5>}.RGB
    Alpha=ComponentMask(None=@n20) {g=true}
   )
   False=@n1.RGB
  ) {parameterName=bIsGradient}
 ]
}
