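# UI-domain master material for BP toast backgrounds, blending two colour sets with a radial gradient over a faint triangle grid.
# Each colour pair mixes at a fixed midpoint, and the grid lightens only the first set toward white.
# The grid uses aspect-corrected screen-aligned UVs, rotated by a parameter-driven fixed angle and tiled.
# An exponential radial gradient with adjustable centre, size and density blends in the second set.
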Material /Game/UI/Foundation/Materials/M_UI_BPToast_Master
graph {
 properties {bUsedWithStaticMesh=false, materialDomain=MD_UI, pixelDepthOffsetMode=PDOM_Legacy, usedShadingModels=Default Lit}
 nodes [
  n1 = VectorParameter() {group=Background, parameterName=Gradient_RGB01a, defaultValue=<rgba 0 0.0056 0.1329 1>}
  n2 = VectorParameter() {group=Background, parameterName=Gradient_RGB01b, defaultValue=<rgba 0 0.0802 0.5972 1>}
  n3 = MaterialFunctionCall(Float2=MaterialFunctionCall() {materialFunction=&ScreenResolution}.Visible Resolution) {materialFunction=&BreakOutFloat2Components}
  n4 = Divide(A=@n3.R, B=Max(A=@n3.G))
  n5 = Constant() {r=1}
  n6 = Rotator(
   Coordinate=Multiply(
    A=AppendVector(A=@n4, B=@n5)
    B=MaterialFunctionCall() {materialFunction=&ScreenAlignedUVs}."X 100%, Y 100%"
   )
   Time=ScalarParameter() {defaultValue=-0.0571, parameterName=Tri Angle}
  ) {speed=6.2832}
  n7 = TextureSample(UVs=Multiply(A=Panner(Coordinate=@n6), B=ScalarParameter() {defaultValue=4, parameterName=Tri Tile Amount})) {texture=/Game/Global/Textures/Default/Masks_and_Gradients/T-TriangleGrid}
  n8 = ScalarParameter() {defaultValue=0.1, parameterName=TriangleDensity, sliderMax=1}
  n9 = ScalarParameter() {defaultValue=12, group=Background Gradient, parameterName=Gradient_Position_X}
  n10 = ScalarParameter() {defaultValue=50, group=Background Gradient, parameterName=Gradient_Position_Y}
  n11 = MaterialFunctionCall(
   CenterPosition=AppendVector(A=Multiply(A=@n9) {constB=0.01}, B=Multiply(A=@n10) {constB=0.01})
   Radius=Multiply(A=ScalarParameter() {defaultValue=70, group=Background Gradient, parameterName=Gradient_Size}) {constB=0.015}
   Density=ScalarParameter() {defaultValue=1, group=Background, parameterName=GradientDensity, sliderMax=1}
  ) {materialFunction=&RadialGradientExponential}
  LinearInterpolate(
   A=LinearInterpolate(A=LinearInterpolate(A=@n1.RGB, B=@n2.RGB), Alpha=Multiply(A=@n7.R, B=@n8))
   B=LinearInterpolate(
    A=VectorParameter() {group=Background, parameterName=Gradient_RGB02a, defaultValue=<rgba 0.1458 0.7784 1 1>}.RGB
    B=VectorParameter() {group=Background, parameterName=Gradient_RGB02b, defaultValue=<rgba 0.0003 0.4793 0.5972 1>}.RGB
   )
   Alpha=Multiply(A=Saturate(None=@n11.RadialGradientExponential))
  )
 ]
}
