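# Translucent UI master material for gradient panels, configured through static switches.
# A linear, reflected-linear or radial gradient mask blends two colors and alphas, with optional hover and selected colors lerped by scalar states.
# Color can add a tiled pixel-size pattern, or switch to a vertical two-tone blend with optional panning triangle layers.
# Alpha optionally multiplies by a box-scale mask texture, then by a global alpha.
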Material /Game/UI/Foundation/Materials/M_UI_GradientPanel
graph {
 properties {
  bUsedWithStaticMesh=false
  blendMode=BLEND_Translucent
  materialDomain=MD_UI
  pixelDepthOffsetMode=PDOM_Legacy
  usedShadingModels=Default Lit
 }
 nodes [
  n1 = VectorParameter() {group=Color, parameterName=RGBA0, defaultValue=<rgba 1 0 0 1>}
  n2 = VectorParameter() {group=Color, parameterName=RGBA1, defaultValue=<rgba 0 1 0 1>}
  n3 = MaterialFunctionCall() {materialFunction=&GetUserInterfaceUV}
  n4 = StaticSwitchParameter(True=@n3.9-Slice UV, False=@n3.9-Slice UV) {parameterName=Box Scale Linear Gradient}
  n5 = Abs(None=Add(A=@n4) {constB=-0.5})
  n6 = ScalarParameter() {defaultValue=1, group=Linear Gradient Settings, parameterName=Rotation}
  n7 = Divide(
   A=StaticComponentMaskParameter(None=Rotator(Coordinate=@n4, Time=@n6)) {defaultR=true, group=Options, parameterName=Gradient Direction}
   B=ScalarParameter() {defaultValue=1, group=Linear Gradient Settings, parameterName=Scale}
  )
  n8 = ScalarParameter() {group=Linear Gradient Settings, parameterName=Position}
  n9 = StaticSwitchParameter(
   True=MaterialFunctionCall(
    Alpha=StaticComponentMaskParameter(None=@n5) {defaultR=true, group=Options, parameterName=Reflected Gradient Direction}
    Min=ScalarParameter() {group=Reflected Gradient Settings, parameterName=RG SS Min}
    Max=ScalarParameter() {defaultValue=1, group=Reflected Gradient Settings, parameterName=RG SS Max}
   ) {materialFunction=&SmoothStep}.Result
   False=MaterialFunctionCall(
    Alpha=Add(A=@n7, B=@n8)
    Min=ScalarParameter() {group=Linear Gradient Settings, parameterName=LG SS Min}
    Max=ScalarParameter() {defaultValue=1, group=Linear Gradient Settings, parameterName=LG SS Max}
   ) {materialFunction=&SmoothStep}.Result
  ) {group=Options, parameterName=Reflected Linear}
  n10 = OneMinus(None=@n9)
  n11 = ScalarParameter() {defaultValue=0.5, group=Sphere Gradient Settings, parameterName=Sphere Gradient Pos U}
  n12 = ScalarParameter() {defaultValue=0.5, group=Sphere Gradient Settings, parameterName=Sphere Gradient Pos V}
  n13 = Power(
   Base=MaterialFunctionCall(
    UVs=TextureCoordinate() {coordinateIndex=4}
    CenterPosition=AppendVector(A=@n11, B=@n12)
    Radius=ScalarParameter() {defaultValue=1, group=Sphere Gradient Settings, parameterName=Sphere Gradient Radius}
    Density=ScalarParameter() {defaultValue=1, group=Sphere Gradient Settings, parameterName=Sphere Gradient Density}
    Invert Density=StaticBoolParameter() {group=Sphere Gradient Settings, parameterName=Sphere Gradient Invert}
   ) {materialFunction=&RadialGradientExponential}.RadialGradientExponential
   Exp=ScalarParameter() {defaultValue=1, group=Sphere Gradient Settings, parameterName=Gradient Contrast}
  )
  n14 = ScalarParameter() {defaultValue=1, group=Sphere Gradient Settings, parameterName=Gradient Mult}
  n15 = Clamp(None=StaticSwitchParameter(True=Multiply(A=@n13, B=@n14), False=StaticSwitchParameter(True=@n10, False=@n9) {group=Options, parameterName=Invert Gradient}) {group=Options, parameterName=Use Sphere Gradient})
  n16 = StaticSwitchParameter(True=@n1.A, False=Clamp(None=LinearInterpolate(A=@n1.A, B=@n2.A, Alpha=@n15))) {group=Options, parameterName=Use Single Color}
  n17 = Multiply(
   A=TextureSampleParameter2D(UVs=TextureCoordinate() {desc=Box Scale}) {parameterName=Box Scale Mask, texture=/Game/UI/Foundation/Items/Textures/Item-MainBG}.A
   B=@n16
  )
  Reroute(None=Multiply(
    A=StaticSwitchParameter(True=@n17, False=@n16) {group=Options, parameterName=Box Scaling}
    B=ScalarParameter() {defaultValue=1, group=Color, parameterName=Global Alpha, sliderMax=1}
   ))
  n18 = MaterialFunctionCall() {materialFunction=&GetUserInterfaceUV}
  n19 = MaterialFunctionCall(Float2=@n18.Pixel Size) {materialFunction=&BreakOutFloat2Components}
  n20 = OneMinus(None=@n19.R)
  n21 = ScalarParameter() {parameterName=Triangle_XShift}
  n22 = AppendVector(A=Divide(A=Add(A=@n20, B=@n21), B=Max(B=@n19.G) {constA=0.1}), B=Constant() {r=1})
  n23 = ScalarParameter() {defaultValue=0.9048, parameterName=tri-L3-Scale}
  n24 = Add(
   A=Multiply(A=@n18.Normalized UV, B=Multiply(A=@n22, B=@n23))
   B=AppendVector(A=Constant(), B=ScalarParameter() {defaultValue=0.019, parameterName=tri-L3-Offset})
  )
  n25 = ScalarParameter() {defaultValue=1, parameterName=ss min}
  n26 = ComponentMask(None=TextureCoordinate()) {r=true}
  n27 = Power(Base=Add(A=OneMinus(None=@n26), B=@n21)) {constExponent=1.25}
  n28 = ScalarParameter() {defaultValue=0.35, parameterName=ss max}
  n29 = MaterialFunctionCall(
   Alpha=TextureSample(UVs=Panner(Coordinate=@n24) {speedX=-0.35}) {samplerType=SAMPLERTYPE_LinearColor, texture=/Game/UI/Foundation/Items/Textures/T-Card-Hero-Anim-Triangle}.B
   Min=LinearInterpolate(B=@n25, Alpha=@n27) {constA=0.25, constB=0.8}
   Max=LinearInterpolate(A=@n28, Alpha=@n27) {constA=0.2, constB=2}
  ) {materialFunction=&SmoothStep}
  n30 = Multiply(A=@n18.Normalized UV, B=Multiply(A=@n22, B=ScalarParameter() {defaultValue=0.5743, parameterName=tri-L2-Scale}))
  n31 = AppendVector(A=Constant(), B=ScalarParameter() {defaultValue=-0.67, parameterName=tri-L2-Offset})
  n32 = TextureSample(UVs=Panner(Coordinate=Add(A=@n30, B=@n31)) {speedX=-0.45}) {samplerType=SAMPLERTYPE_LinearColor, texture=&T-Card-Hero-Anim-Triangle}
  n33 = LinearInterpolate(B=@n25, Alpha=@n27) {constA=0.25, constB=0.9}
  n34 = LinearInterpolate(A=@n28, Alpha=@n27) {constA=0.1, constB=2}
  n35 = VectorParameter() {parameterName=RGB_TwoTone_L, defaultValue=<rgba 0.6444 0.428 0.7552 1>}
  n36 = Multiply(
   A=LinearInterpolate(
    A=Multiply(A=@n29.Result) {constB=0.125}
    Alpha=MaterialFunctionCall(Alpha=@n32.G, Min=@n33, Max=@n34) {materialFunction=&SmoothStep}.Result
   ) {constB=0.25}
   B=Multiply(A=ScalarParameter() {defaultValue=1, parameterName=TriangleColorBrightness}, B=@n35.RGB)
  )
  n37 = Multiply(A=ComponentMask(None=TextureCoordinate()) {g=true}) {constB=2}
  n38 = VectorParameter() {group=Color, parameterName=RGBA0_1, defaultValue=<rgba 1 0 0 1>}
  n39 = ScalarParameter() {parameterName=RGBA0 SS Min}
  n40 = ScalarParameter() {parameterName=RGBA0 SS Max}
  n41 = ScalarParameter() {parameterName=RGBA0 Rotation}
  n42 = LinearInterpolate(
   A=StaticSwitchParameter(
    True=MaterialFunctionCall!(In, RGBA0=@n1.RGB, RGBA1=@n38.RGB, SS Min=@n39, SS Max=@n40, Gradient Rotation=@n41) {materialFunction=&/Mf-UILinearGradientColor}.Linear Gradient
    False=@n1.RGB
   ) {parameterName=RGBA0 Split Color}
   B=@n2.RGB
   Alpha=@n15
  )
  n43 = LinearInterpolate(
   A=@n42
   B=LinearInterpolate(A=@n35.RGB, B=VectorParameter() {parameterName=RGB_TwoTone_R, defaultValue=<rgba 0.1414 0.6615 0.1588 1>}.RGB, Alpha=@n15)
   Alpha=Power(Base=Subtract(A=@n37)) {constExponent=0.5}
  )
  n44 = LinearInterpolate(
   A=VectorParameter() {group=Color, parameterName=RGBA 0 Hover, defaultValue=<rgba 1 0 1 1>}.RGB
   B=VectorParameter() {group=Color, parameterName=RGBA 1 Hover, defaultValue=<rgba 0 1 1 1>}.RGB
   Alpha=@n15
  )
  n45 = ScalarParameter() {group=Interaction, parameterName=HoverState, sliderMax=1}
  n46 = VectorParameter() {group=Color, parameterName=RGBA 0 Selected, defaultValue=<rgba 0 0.1 0 1>}
  n47 = VectorParameter() {group=Color, parameterName=RGBA 1 Selected, defaultValue=<rgba 0.1 0 0 1>}
  n48 = StaticSwitchParameter(
   True=LinearInterpolate(
    A=LinearInterpolate(A=@n42, B=@n44, Alpha=@n45)
    B=LinearInterpolate(A=@n46.RGB, B=@n47.RGB, Alpha=@n15)
    Alpha=ScalarParameter() {group=Interaction, parameterName=SelectedState, sliderMax=1}
   )
   False=StaticSwitchParameter(True=@n1.RGB, False=@n42) {group=Options, parameterName=Use Single Color}
  ) {group=Options, parameterName=Use States}
  n49 = VectorParameter() {parameterName=Pattern Color, defaultValue=<rgba 0 0 0 1>}
  n50 = MaterialFunctionCall() {materialFunction=&GetUserInterfaceUV}
  n51 = Multiply(A=Divide(A=@n50.Pixel Size, B=ScalarParameter() {defaultValue=16, parameterName=Pattern Pixel Size}), B=@n50.Normalized UV)
  n52 = LinearInterpolate(A=@n48, B=@n49.RGB, Alpha=Multiply(A=TextureSampleParameter2D(UVs=@n51) {group=Texture, parameterName=Pattern}.R, B=@n49.A))
  StaticSwitchParameter(
   True=StaticSwitchParameter(True=Add(A=@n36, B=@n43), False=@n43) {parameterName=UseTriangleScroll}
   False=Multiply(
    A=ScalarParameter() {defaultValue=1, parameterName=Multiply Base}
    B=StaticSwitchParameter(True=@n52, False=@n48) {group=Options, parameterName=Use Texture Pattern}
   )
  ) {parameterName=UseTwoTone_Vertical}
 ]
}
