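# Translucent UI material that draws a procedural triangle with configurable fill color and an optional border.
# Builds three edge gradients from UVs rotated a third of a turn apart, each with rotation and shift parameters.
# Sharpness scales and saturates them into masks, though the product reads R twice and B, leaving G unused.
# The HasBorder static switch adds a thickness-inset second mask to blend border and fill color and alpha.
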Material /Game/UI/Foundation/Materials/M_UI_Triangle
graph {
 properties {
  bUsedWithStaticMesh=false
  blendMode=BLEND_Translucent
  materialDomain=MD_UI
  pixelDepthOffsetMode=PDOM_Legacy
  usedShadingModels=Default Lit
 }
 nodes [
  n1 = VectorParameter() {group=Triangle, parameterName=TriangleBorderRGBA, defaultValue=<rgba 0 1 0 1>}
  n2 = VectorParameter() {group=Triangle, parameterName=TriangleFillRGBA, defaultValue=<rgba 1 0 0 0.5>}
  n3 = TextureCoordinate()
  n4 = Add(
   A=Rotator(Coordinate=@n3, Time=ScalarParameter() {defaultValue=0.0312, group=Triangle, parameterName=Angle_Rotation_A}) {speed=6.2832}
   B=Multiply(A=ScalarParameter() {defaultValue=0.25, group=Triangle, parameterName=Angle_Shift_A}) {constB=-1}
  ) {constB=-0.5}
  n5 = ScalarParameter() {defaultValue=0.0312, group=Triangle, parameterName=Angle_Rotation_B}
  n6 = Add(
   A=Rotator(Coordinate=@n3, Time=Add(A=@n5) {constB=0.3333}) {speed=6.2832}
   B=Multiply(A=ScalarParameter() {defaultValue=0.25, group=Triangle, parameterName=Angle_Shift_B}) {constB=-1}
  ) {constB=-0.5}
  n7 = Rotator(Coordinate=@n3, Time=Add(A=ScalarParameter() {defaultValue=0.0312, group=Triangle, parameterName=Angle_Rotation_C}) {constB=0.6667}) {speed=6.2832}
  n8 = Multiply(A=ScalarParameter() {defaultValue=0.25, group=Triangle, parameterName=Angle_Shift_C}) {constB=-1}
  n9 = AppendVector(
   A=AppendVector(A=ComponentMask(None=@n4) {g=true}, B=ComponentMask(None=@n6) {g=true})
   B=ComponentMask(None=Add(A=@n7, B=@n8) {constB=-0.5}) {g=true}
  )
  n10 = Add(A=@n9, B=Multiply(A=ScalarParameter() {defaultValue=0.0625, parameterName=Border_Thickness}) {constB=-1})
  n11 = ScalarParameter() {defaultValue=32, group=Triangle, parameterName=Sharpness}
  n12 = MaterialFunctionCall(Float3=Saturate(None=Multiply(A=@n10, B=@n11))) {materialFunction=&BreakOutFloat3Components}
  n13 = Saturate(None=DotProduct(A=DotProduct(A=@n12.R, B=@n12.R), B=@n12.B))
  n14 = MaterialFunctionCall(Float3=Saturate(None=Multiply(A=@n9, B=@n11))) {materialFunction=&BreakOutFloat3Components}
  n15 = Saturate(None=DotProduct(A=DotProduct(A=@n14.R, B=@n14.R), B=@n14.B))
  StaticSwitchParameter(True=LinearInterpolate(B=LinearInterpolate(A=@n1.A, B=@n2.A, Alpha=@n13), Alpha=@n15), False=LinearInterpolate(B=@n2.A, Alpha=@n15)) {parameterName=HasBorder}
  StaticSwitchParameter(True=LinearInterpolate(A=@n1.RGB, B=@n2.RGB, Alpha=@n13), False=@n2.RGB) {parameterName=HasBorder}
 ]
}
