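# M-FontShadow is a translucent UI material that, by its name, serves as a shadow for fonts.
# It runs the second UV channel through two sine waves of slightly different periods, multiplying the horizontal and vertical results via a dot product.
# The product is sharpened by a power, clamped and used to blend up to a partial maximum, forming a soft procedural mask alongside a constant white color.
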Material /Game/UI/Foundation/Materials/M-FontShadow
graph {
 properties {
  bUsedWithStaticMesh=false
  blendMode=BLEND_Translucent
  materialDomain=MD_UI
  pixelDepthOffsetMode=PDOM_Legacy
  usedShadingModels=Default Lit
 }
 nodes [
  Constant3Vector() {constant=<rgba 1 1 1 1>}
  n1 = TextureCoordinate() {coordinateIndex=1}
  n2 = DotProduct(A=ComponentMask(None=Sine(None=@n1) {period=2}) {r=true}, B=ComponentMask(None=Sine(None=@n1) {period=2.1252}) {g=true})
  LinearInterpolate(Alpha=Clamp(None=Power(Base=@n2) {constExponent=1.5})) {constB=0.65}
 ]
}
