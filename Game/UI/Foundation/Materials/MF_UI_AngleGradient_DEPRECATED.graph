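# Deprecated UI material function that outputs an angular gradient around the UV center, with optional rotation.
# UVs and Rotation inputs, defaulting to standard texture coordinates and zero, feed CustomRotator.
# The rotated components are swapped and inverted, then remapped from 0–1 to -1–1.
# VectorToRadialValue converts that centered vector to an angle, which is the only output.
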MaterialFunction /Game/UI/Foundation/Materials/MF_UI_AngleGradient_DEPRECATED
graph {nodes [
  n1 = TextureCoordinate()
  n2 = Constant()
  n3 = MaterialFunctionCall(Float2=MaterialFunctionCall(
    UVs=FunctionInput(Preview=@n1) {bUsePreviewValueAsDefault=true, inputName=UVs, inputType=FunctionInput_Vector2}
    "Rotation Angle (0-1)"=FunctionInput(Preview=@n2) {bUsePreviewValueAsDefault=true, inputName=Rotation, inputType=FunctionInput_Scalar}
   ) {materialFunction=&CustomRotator}.Rotated Values) {materialFunction=&BreakOutFloat2Components}
  n4 = AppendVector(A=OneMinus(None=@n3.G), B=OneMinus(None=@n3.R))
  FunctionOutput(None=MaterialFunctionCall(Vector or UVs=ConstantBiasScale(None=@n4) {bias=-0.5, scale=2}) {materialFunction=&VectorToRadialValue}.Vector Converted to Angle)
 ]}
