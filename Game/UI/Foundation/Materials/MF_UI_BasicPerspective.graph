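# Material function for UI that applies a simple projective perspective distortion to input UVs.
# Optionally flipped X and Y are scaled and sheared by the FixedScaleFactor and Slide inputs, then divided by a shared denominator built from Persp_X and Persp_Y.
# A Ceil/Saturate chain builds a branchless mask for the 0–1 range.
# It outputs clamped and unclamped UVs, the mask, and an optional texture's RGB multiplied by that mask.
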MaterialFunction /Game/UI/Foundation/Materials/MF_UI_BasicPerspective
graph {nodes [
  n1 = FunctionInput(Preview=Constant()) {bUsePreviewValueAsDefault=true, inputName=FixedScaleFactor_X, inputType=FunctionInput_Scalar}
  n2 = FunctionInput(Preview=TextureCoordinate()) {bUsePreviewValueAsDefault=true, inputName=UV, inputType=FunctionInput_Vector2, sortPriority=-1}
  n3 = ComponentMask(None=@n2) {r=true}
  n4 = StaticBool()
  n5 = StaticSwitch(
   True=OneMinus(None=@n3)
   False=@n3
   Value=FunctionInput(Preview=@n4) {bUsePreviewValueAsDefault=true, inputName=Flip_X, inputType=FunctionInput_StaticBool, sortPriority=3}
  )
  n6 = ComponentMask(None=@n2) {g=true}
  n7 = StaticSwitch(
   True=OneMinus(None=@n6)
   False=@n6
   Value=FunctionInput(Preview=@n4) {bUsePreviewValueAsDefault=true, inputName=Flip_Y, inputType=FunctionInput_StaticBool, sortPriority=3}
  )
  n8 = Multiply(A=FunctionInput(Preview=Constant()) {inputName=Slide_X, inputType=FunctionInput_Scalar, sortPriority=2}, B=@n7)
  n9 = FunctionInput(Preview=Constant() {r=1}) {inputName=Persp_Y, inputType=FunctionInput_Scalar, sortPriority=1}
  n10 = FunctionInput(Preview=Constant() {r=1}) {inputName=Persp_X, inputType=FunctionInput_Scalar, sortPriority=1}
  n11 = Add(A=Add(A=Multiply(A=@n9, B=@n5), B=Multiply(A=@n7, B=@n10)))
  n12 = Divide(A=Add(A=Multiply(A=@n1, B=@n5), B=Add(A=@n5, B=@n8)), B=@n11)
  n13 = Multiply(A=@n5, B=FunctionInput(Preview=Constant()) {inputName=Slide_Y, inputType=FunctionInput_Scalar, sortPriority=2})
  n14 = FunctionInput(Preview=Constant()) {bUsePreviewValueAsDefault=true, inputName=FixedScaleFactor_Y, inputType=FunctionInput_Scalar}
  n15 = Divide(A=Add(A=Add(A=@n7, B=@n13), B=Multiply(A=@n7, B=@n14)), B=@n11)
  n16 = MaterialFunctionCall(X=@n12, Y=@n15) {materialFunction=&MakeFloat2}
  FunctionOutput(None=@n16.Result) {outputName=UnclampedFinalUvs}
  n17 = FunctionInput(Preview=TextureObject() {texture=/Game/Global/Textures/Default/Blanks/T_Blank_White_ZeroAlpha_sRGBColor}) {bUsePreviewValueAsDefault=true, description=optional texture, inputName=Texture, inputType=FunctionInput_Texture2D}
  n18 = Ceil(None=@n12)
  n19 = OneMinus(None=Ceil(None=Subtract(A=@n12)))
  n20 = Ceil(None=@n15)
  n21 = OneMinus(None=Ceil(None=Subtract(A=@n15)))
  n22 = Multiply(A=Multiply(A=Saturate(None=@n18), B=Saturate(None=@n19)), B=Multiply(A=Saturate(None=@n20), B=Saturate(None=@n21)))
  FunctionOutput(None=Multiply(A=TextureSample(UVs=@n16.Result, Tex=@n17).RGB, B=@n22)) {outputName="FinalColor (if texture)"}
  FunctionOutput(None=Saturate(None=@n16.Result)) {outputName=FinalUvs}
  FunctionOutput(None=@n22) {outputName=0 or 1 Mask}
 ]}
