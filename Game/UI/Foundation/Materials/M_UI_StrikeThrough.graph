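# M_UI_StrikeThrough is a masked UI material that draws a colored diagonal strike-through line, red by default.
# DrawLine-2D builds a square-cornered line between two inset points on opposite sides, built with MakeFloat2 from constants.
# The LineWidth parameter sets its thickness, and the result is multiplied by the LineColor parameter to tint it.
# A scalar parameter named Y is present but feeds nothing.
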Material /Game/UI/Foundation/Materials/M_UI_StrikeThrough
graph {
 properties {bUsedWithStaticMesh=false, blendMode=BLEND_Masked, materialDomain=MD_UI, pixelDepthOffsetMode=PDOM_Legacy, usedShadingModels=Default Lit}
 nodes [
  n1 = Constant() {r=1}
  n2 = Constant() {r=0.2}
  n3 = Add(A=Constant(), B=@n2) {constB=0}
  n4 = Constant()
  n5 = Subtract(A=Constant() {r=1}, B=@n2)
  Multiply(
   A=VectorParameter() {parameterName=LineColor, defaultValue=<rgba 0.7529 0.0931 0.0704 1>}.RGB
   B=MaterialFunctionCall!(
    Coordinates
    P0=MaterialFunctionCall(X=@n1, Y=@n3) {materialFunction=&MakeFloat2}.Result
    P1=MaterialFunctionCall(X=@n4, Y=@n5) {materialFunction=&MakeFloat2}.Result
    Line Width=ScalarParameter() {defaultValue=0.025, parameterName=LineWidth}
    Edge Falloff=Constant() {r=1}
    Square Corners?=StaticBool() {value=true}
   ) {materialFunction=/Engine/Functions/Engine_MaterialFunctions01/Coordinates/DrawLine-2D}.Result
  )
  ScalarParameter() {defaultValue=0.6, parameterName=Y}
 ]
}
